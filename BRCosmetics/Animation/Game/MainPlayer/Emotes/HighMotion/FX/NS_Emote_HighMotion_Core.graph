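# Niagara system for the glowing energy core effect of the HighMotion player emote, in a mostly purple palette.
# Orbs and debris spawn on buildup curves in widening spheres and converge by position lerp on a point just ahead of the origin.
# An anamorphic flare and stretched cylinder mesh pulse with system intensity and age, while delayed spark bursts spray from the core.
# AttractedSparks and FN_Debug emitters are disabled.
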NiagaraSystem /BRCosmetics/Animation/Game/MainPlayer/Emotes/HighMotion/FX/NS_Emote_HighMotion_Core
dependencies {
 usedDataInterfaces [
  &NiagaraDataInterfaceCurve
  &NiagaraDataInterfaceColorCurve
  &NiagaraDataInterfaceVector2DCurve
  &NiagaraDataInterfaceArrayFloat3
  &NiagaraDataInterfaceArrayFloat2
  &NiagaraDataInterfaceCamera
  &NiagaraDataInterfaceMeshRendererInfo
 ]
 usedDynamicInputs [&Lerp_Vector2, &Sine]
 usedModules [
  /Niagara/Modules/System/SystemLifeCycle
  &SystemState
  &self:SystemSpawnScript.SystemScriptSource.SystemScriptGraph.NiagaraNodeAssignment_0.SetVariables_#1
 ]
}
emitters {
 AnamorphicFlare {
  emitterUpdateScript [
   &EmitterState(
    Inactive Response=::NewEnumerator1
    Life Cycle Mode=::NewEnumerator1
    Loop Behavior=::NewEnumerator1
    Loop Duration Mode=::NewEnumerator1
    UseLoopDelay=false
   )
   &SpawnBurst_Instantaneous(Use Spawn Probability=false, UseLoopCountLimit=false)
  ]
  particleSpawnScript [
   &FN_SetSpawnGroup(Number of Spawn Groups=1)
   &InitializeParticle(
    Color=<rgba 17.2 4 40 0.4>
    Position Offset=&SelectVectorFromArray(
     Direct Array Index=&ReturnExecIndex()
     Vector Selection Array=DataInterface() {propertyValues {FloatData [<36 0 0>, <35 0 0>]}}
    )
    Position Offset Coordinate Space=::Local
    Sprite Size=&SelectVector2DFromArray(
     Direct Array Index=&ReturnExecIndex()
     Vector 2D Selection Array=DataInterface() {propertyValues {FloatData [<200 10>, <200 20>]}}
    )
    Sprite Size Mode=::NewEnumerator4
    UsePositionOffset=true
    Write Material Random=true
   )
  ]
  particleUpdateScript [
   &ParticleState(Kill Particles When Lifetime Has Elapsed=false)
   &CameraOffset(Camera Offset Amount=&FloatFromCurve(FloatCurve=DataInterface() {propertyValues {Curve {keys [<at 0.25 10>,<at 0.45 0>]}}}, CurveIndex=$System.NormalizedLoopAge)) {enabled=false}
   &CameraOffset(Camera Offset Amount=10) {module=CameraOffset001}
   &ScaleColor(Scale Alpha=$System.IntensityOverTime, ScaleA=true, ScaleRGB=false)
   &ScaleColor(
    Scale Alpha=&FloatFromCurve(
     FloatCurve=DataInterface() {propertyValues {Curve {keys [
         {arriveTangent=4.1785, interpMode=RCIM_Cubic, tangentMode=RCTM_User, time=0.2, value=0}
         {interpMode=RCIM_Cubic, leaveTangent=0.3975, tangentMode=RCTM_User, time=3, value=1}
        ]}}}
     Scale Curve=10
    )
    ScaleA=true
    ScaleRGB=true
   ) {module=ScaleColor001}
   &ScaleColor(
    Scale Alpha=&FloatFromCurve(FloatCurve=DataInterface() {propertyValues {Curve {keys [<at 0.2 1>,<at 0.8 0>]}}}, CurveIndex=$System.NormalizedLoopAge, Scale Curve=2)
    ScaleA=true
    ScaleRGB=true
   ) {module=ScaleColor002}
   &ScaleColor(
    Scale Alpha=/Game/Effects/Niagara/DynamicInputs/Math/FN_RemapDotProduct(
     SystemAxis:Vector3f=/Game/Effects/Niagara/DynamicInputs/Engine/FN_GetSystemAxis(Axis:ENiagaraOrientationAxis=::XAxis, Flip Axis:NiagaraBool=false)
     ComparisonAxis:Vector3f=&ReturnCameraVector(Camera Vectors=::NewEnumerator1, Coordinate Space=::Simulation)
     SourceMin:NiagaraFloat=0
     SourceMax:NiagaraFloat=1
     DestMin:NiagaraFloat=0
     DestMax:NiagaraFloat=10
    )
    ScaleA=true
    ScaleRGB=true
   ) {enabled=false, module=ScaleColor003}
  ]
  rendererClasses [&NiagaraSpriteRendererProperties]
  renderers [{rendererClass=&NiagaraSpriteRendererProperties}]
 }
 AttractedOrbs {
  emitterUpdateScript [
   &EmitterState(Life Cycle Mode=::NewEnumerator1, Loop Behavior=::NewEnumerator1, Loop Duration=3.2, UseLoopDelay=true)
   /SetVariables(Emitter.BuildupCurve:NiagaraFloat=&FloatFromCurve(
     FloatCurve=DataInterface() {propertyValues {Curve {keys [
         {arriveTangent=0.0145, interpMode=RCIM_Cubic, leaveTangent=0.0145, tangentMode=RCTM_User, time=0, value=0}
         {arriveTangent=2.5625, interpMode=RCIM_Cubic, leaveTangent=2.5625, tangentMode=RCTM_User, time=0.8, value=1}
         {interpMode=RCIM_Cubic, leaveTangent=1.25, time=0.8373, value=0}
        ]}}}
     CurveIndex=$Emitter.NormalizedLoopAge
    )) {isSetParameters=true}
   /SetVariables(Emitter.BuildupCurveInfinite:NiagaraFloat=&FloatFromCurve(
     FloatCurve=DataInterface() {propertyValues {Curve {keys [
         {arriveTangent=0.0145, interpMode=RCIM_Cubic, leaveTangent=0.0145, tangentMode=RCTM_User, time=0, value=0}
         {arriveTangent=2.5625, interpMode=RCIM_Cubic, leaveTangent=2.5625, tangentMode=RCTM_User, time=1, value=1}
        ]}}}
     CurveIndex=$Emitter.NormalizedLoopAge
    )) {isSetParameters=true}
   &SpawnRate(SpawnRate=200, Use Spawn Probability=false) {enabled=false, module=SpawnRate001}
   &SpawnRate(SpawnRate=&Multiply_Float(A=$Emitter.BuildupCurve, B=40), Use Spawn Probability=false)
  ]
  particleSpawnScript [
   /SetVariables(Particles.PerParticleRandom:NiagaraFloat=&RandomRangeFloat(Minimum=0, Maximum=1)) {isSetParameters=true, module=SetVariables_#2}
   &InitializeParticle(
    Color=&MultiplyLinearColor(
     Linear Color=&RandomRangeLinearColor(Minimum=<rgba 0.6979 0.375 1 1>, Maximum=<rgba 0.15 0 1 1>)
     Scale Factor=&MakeFloat4FromVectorAndFloat(
      "Vector (RGB)"=&VectorFromFloat(Value=&FloatFromCurve(
        FloatCurve=DataInterface() {propertyValues {Curve {keys [{tangentMode=RCTM_User, time=0.4, value=0}, {tangentMode=RCTM_User, time=1, value=1}]}}}
        CurveIndex=$Emitter.BuildupCurve
        Scale Curve=15
       ))
      "Float (Alpha)"=1
     )
    )
    Lifetime=&FloatFromCurve(FloatCurve=DataInterface() {propertyValues {Curve {keys [<at 0 1.5>,<at 1 0>]}}}, CurveIndex=$Emitter.BuildupCurveInfinite)
    Sprite Size=&MakeVector2D(
     X=&FloatFromCurve(FloatCurve=DataInterface() {propertyValues {Curve {keys [<at 0 1>,<at 1 0>]}}}, CurveIndex=$Emitter.BuildupCurveInfinite, Scale Curve=10)
     Y=&FloatFromCurve(FloatCurve=DataInterface() {propertyValues {Curve {keys [<at 0.4 1>,<at 1 15>]}}}, CurveIndex=$Emitter.BuildupCurveInfinite, Scale Curve=14)
    )
    Sprite Size Mode=::NewEnumerator4
    UsePositionOffset=false
    Write Material Random=true
   )
   &SkeletalMeshLocation(Bone Sampling Mode=::NewEnumerator0, MaskSpawnGroup=false, OverrideRandomSeed=false, Skeletal Mesh=$User.SkeletalMesh) {enabled=false, module=SkeletalMeshLocation002}
   &AddVelocity(
    Cone Angle=33
    Cone Axis=<0 0.25 1>
    Module.UseConeFalloff=false
    Module.UseVelDistribution=false
    Override Seed=false
    Velocity Mode=::NewEnumerator2
    Velocity Speed=&FloatFromCurve(
     FloatCurve=DataInterface() {propertyValues {Curve {keys [<at 0 0.5>,<at 0.333 1>,<at 0.666 0.1>,<at 3 0>]}}}
     CurveIndex=$Emitter.LoopedAge
     Scale Curve=&RandomRangeFloat(Minimum=444, Maximum=999)
    )
   ) {enabled=false}
   &ShapeLocation(
    MaskSpawnGroup=false
    Offset=&MakeVector(
     X=&FloatFromCurve(
      FloatCurve=DataInterface() {propertyValues {Curve {keys [{tangentMode=RCTM_User, time=0.1, value=0.2}, {tangentMode=RCTM_User, time=1, value=1}]}}}
      CurveIndex=$Emitter.BuildupCurve
      Scale Curve=150
     )
     Y=0
     Z=0
    )
    Offset Mode=::NewEnumerator0
    Override Seed=false
    Rotation Mode=::NewEnumerator2
    Sphere Radius=&FloatFromCurve(
     FloatCurve=DataInterface() {propertyValues {Curve {keys [{tangentMode=RCTM_User, time=0, value=0.1}, {tangentMode=RCTM_User, time=1, value=1}]}}}
     CurveIndex=$Emitter.BuildupCurveInfinite
     Scale Curve=100
    )
    UseOwnerScale=true
   )
   /SetVariables(Particles.SpriteAlignment=&NormalizeVector(
     Vector To Normalize=&ConvertPositionToVector(
      Input Position=&AddVectorToPosition(Position=$Particles.Position, Vector=<0 0 0>)
      Conversion Operation:ENiagara_LWCConvertPosToVec=::NewEnumerator0
     )
     Safe Normalize=true
     Safe Fallback Threshold=0
     Safe Fallback Vector=<0 0 0.001>
    )) {isSetParameters=true}
   &AddVelocity(
    Cone Angle=20
    Cone Axis=<0 1 0>
    Module.UseConeFalloff=false
    Module.UseVelDistribution=false
    Override Seed=false
    Rotation Angle Type:ENiagara_AngleInput=::NewEnumerator1
    Rotation Mode=::NewEnumerator2
    Velocity Mode=::NewEnumerator2
    Velocity Speed=&FloatFromCurve(
     FloatCurve=DataInterface() {propertyValues {Curve {keys [
         {interpMode=RCIM_Cubic, tangentMode=RCTM_User, time=0, value=0}
         {arriveTangent=1.4946, interpMode=RCIM_Cubic, leaveTangent=1.4946, tangentMode=RCTM_User, time=1, value=1}
        ]}}}
     CurveIndex=$Emitter.NormalizedLoopAge
     Scale Curve=-400
    )
    Yaw / Pitch / Roll:Vector3f=<-0.25 0 0>
   ) {enabled=false, module=AddVelocity001}
   /Niagara/Modules/Update/Velocity/VortexVelocity(Module.Use Influence Falloff=false, Velocity Amount=4000, Vortex Axis:Vector3f=<0 -1 0>) {enabled=false}
   &AddVelocity(
    Module.Use Falloff Distance=false
    Origin Offset=<0 -15 0>
    Override Seed=false
    Velocity Mode=::NewEnumerator1
    Velocity Speed=&RandomRangeFloat(Minimum=200, Maximum=300)
   ) {enabled=false, module=AddVelocity002}
   &SkeletalMeshLocation(Bone Sampling Mode=::NewEnumerator4, MaskSpawnGroup=false, Skeletal Mesh=$User.SkeletalMesh) {enabled=false}
   &ApplyInitialForces() {enabled=false}
  ]
  particleUpdateScript [
   &ParticleState()
   &AccelerationForce() {enabled=false}
   &Drag(Drag=8, Use Linear Drag=true, Use Rotational Drag=false) {enabled=false}
   /Niagara/Modules/Update/Forces/CurlNoiseForce(
    Noise Frequency=0.1
    Noise Strength=&FloatFromCurve(FloatCurve=DataInterface() {propertyValues {Curve {keys [<at 0 1>,<at 1 0>]}}}, Scale Curve=1000)
    Pan Noise=true
    Pan Noise Field=<0 0 1>
   ) {enabled=false}
   &AccelerationForce(Acceleration=<0 0 250>) {enabled=false, module=AccelerationForce001}
   &VortexForce(Origin Pull Amount=40, UseOriginOffset=false, Vortex Force Amount=800) {enabled=false}
   &PointAttractionForce(Attraction Radius=400, Attraction Strength=100, Kill Within Radius=false, Module.Use Falloff=true, UseAttractorOffset=false) {enabled=false}
   &SolveForcesAndVelocity(Clamp Velocity=false, Limit Acceleration=false)
   /SetVariables(Particles.Position=&LerpPosition(A=$Particles.Position, B=<30 0 0>, Alpha=$Particles.NormalizedAge)) {isSetParameters=true, module=SetVariables_#3}
   &ScaleColor(
    Scale Alpha=&FloatFromCurve(FloatCurve=DataInterface() {propertyValues {Curve {keys [
         {arriveTangent=4.1785, interpMode=RCIM_Cubic, leaveTangent=4.1785, tangentMode=RCTM_User, time=0, value=0}
         {arriveTangent=0.0224, interpMode=RCIM_Cubic, leaveTangent=0.0224, tangentMode=RCTM_User, time=1, value=1}
        ]}}})
    Scale RGB=&MakeVectorFromLinearColorRGB(Color=&HueShiftLinearColor(
      Hue=0.25
      Saturation=1
      Value=$System.IntensityOverTime
      Linear Color=&ColorFromCurve(ColorCurve=DataInterface() {propertyValues {
         AlphaCurve {keys [<at 0 1 RCIM_Cubic>]}
         BlueCurve {keys [<at 0.0265 0>,<at 0.1089 0 RCIM_Cubic>,<at 1 0 RCIM_Cubic>]}
         GreenCurve {keys [<at 0.0265 0>,<at 0.1089 1.8333 RCIM_Cubic>,<at 1 0 RCIM_Cubic>]}
         RedCurve {keys [<at 0.0265 0.2>, {arriveTangent=0.8, interpMode=RCIM_Cubic, time=0.1089, value=10}, <at 1 1 RCIM_Cubic>]}
        }})
     ))
    ScaleA=true
    ScaleRGB=true
   ) {enabled=false}
   &ScaleSpriteSizeBySpeed(Max Scale Factor=<1 10>, Velocity Threshold=20) {enabled=false, module=ScaleSpriteSizeBySpeed001}
   &ScaleSpriteSizeBySpeed(
    Max Scale Factor=<1.666 13>
    Min Scale Factor=<1.333 1.333>
    Sample Scale Factor By Curve=true
    Scale Factor Curve=DataInterface() {propertyValues {Curve {keys [<at 0 0>, {arriveTangent=1.4989, interpMode=RCIM_Cubic, tangentMode=RCTM_Break, time=1, value=1}]}}}
    Velocity Threshold=999
   ) {enabled=false}
   &ScaleSpriteSize(
    Uniform Curve Scale=0.18
    Uniform Curve Sprite Scale=DataInterface() {propertyValues {Curve {keys [
        {tangentMode=RCTM_User, time=0, value=1}
        {arriveTangent=-0.0405, interpMode=RCIM_Cubic, leaveTangent=-0.0405, tangentMode=RCTM_User, time=0.7009, value=1}
        {arriveTangent=-8.0089, interpMode=RCIM_Cubic, leaveTangent=-8.0089, tangentMode=RCTM_User, time=1, value=0}
       ]}}}
   )
   &ScaleSpriteSize(Scale Factor=&MakeVector2D(X=1.4, Y=1.4), Scale Sprite Size Mode=::NewEnumerator2) {module=ScaleSpriteSize001}
   &ScaleColor(
    Scale Alpha=&FloatFromCurve(FloatCurve=DataInterface() {propertyValues {Curve {keys [
         {interpMode=RCIM_Cubic, tangentMode=RCTM_User, time=0, value=0}
         {interpMode=RCIM_Cubic, tangentMode=RCTM_User, time=0.1, value=1}
         {interpMode=RCIM_Cubic, tangentMode=RCTM_User, time=0.5, value=0}
        ]}}})
    ScaleA=true
    ScaleRGB=true
   ) {module=ScaleColor001}
   &ScaleColor(
    Scale Alpha=&VectorLength(Vector=/Niagara/DynamicInputs/Math/Clamp_Vector(
      Vector:Vector3f=&Subtract_Position(A=$Particles.Position, B=<43 0 0>)
      Min:Vector3f=<0 0 0>
      Max:Vector3f=<1 1 1>
     ))
    ScaleA=true
    ScaleRGB=true
   ) {enabled=false, module=ScaleColor002}
  ]
  rendererClasses [&NiagaraSpriteRendererProperties]
  renderers [{rendererClass=&NiagaraSpriteRendererProperties}]
 }
 AttractedSparks {
  bEnabled=false
  emitterUpdateScript [
   &EmitterState(Life Cycle Mode=::NewEnumerator1, Loop Behavior=::NewEnumerator1, Loop Delay=1.5, Loop Duration=1, UseLoopDelay=true)
   &SpawnRate(SpawnRate=200, Use Spawn Probability=false) {enabled=false, module=SpawnRate001}
   &SpawnBurst_Instantaneous(Spawn Count=44, Use Spawn Probability=false, UseLoopCountLimit=false) {enabled=false}
   &SpawnRate(
    SpawnRate=&FloatFromCurve(
     FloatCurve=DataInterface() {propertyValues {Curve {keys [<at 0 0>,<at 1 3>,<at 1.01 0>]}}}
     CurveIndex=$Emitter.NormalizedLoopAge
     Scale Curve=10
    )
    Use Spawn Probability=false
   )
  ]
  particleSpawnScript [
   /SetVariables(Particles.PerParticleRandom:NiagaraFloat=&RandomRangeFloat(Minimum=0, Maximum=1)) {isSetParameters=true, module=SetVariables_#2}
   &InitializeParticle(
    Color Mode=::NewEnumerator0
    Lifetime=&Lerp_Float(A=0.2, B=0.65, Alpha=$Particles.PerParticleRandom)
    Sprite Size=<2 6>
    Sprite Size Mode=::NewEnumerator4
    UsePositionOffset=false
    Write Material Random=true
   )
   &SkeletalMeshLocation(Bone Sampling Mode=::NewEnumerator0, MaskSpawnGroup=false, OverrideRandomSeed=false, Skeletal Mesh=$User.SkeletalMesh) {enabled=false, module=SkeletalMeshLocation002}
   &AddVelocity(
    Cone Angle=33
    Cone Axis=<0 0.25 1>
    Module.UseConeFalloff=false
    Module.UseVelDistribution=false
    Override Seed=false
    Velocity Mode=::NewEnumerator2
    Velocity Speed=&FloatFromCurve(
     FloatCurve=DataInterface() {propertyValues {Curve {keys [<at 0 0.5>,<at 0.333 1>,<at 0.666 0.1>,<at 3 0>]}}}
     CurveIndex=$Emitter.LoopedAge
     Scale Curve=&RandomRangeFloat(Minimum=444, Maximum=999)
    )
   ) {enabled=false}
   &ShapeLocation(
    MaskSpawnGroup=false
    Offset=<100 0 0>
    Offset Mode=::NewEnumerator0
    Override Seed=false
    Rotation Mode=::NewEnumerator2
    Sphere Radius=60
    UseOwnerScale=true
   )
   &AddVelocity(
    Cone Angle=20
    Cone Axis=<0 1 0>
    Module.UseConeFalloff=false
    Module.UseVelDistribution=false
    Override Seed=false
    Rotation Angle Type:ENiagara_AngleInput=::NewEnumerator1
    Rotation Mode=::NewEnumerator2
    Velocity Mode=::NewEnumerator2
    Velocity Speed=&RandomRangeFloat(Minimum=400, Maximum=700)
    Yaw / Pitch / Roll:Vector3f=<-0.25 0 0>
   ) {enabled=false, module=AddVelocity001}
   &VortexVelocity(Module.Use Influence Falloff=false, Velocity Amount=4000, Vortex Axis:Vector3f=<0 -1 0>) {enabled=false}
   &AddVelocity(
    Module.Use Falloff Distance=false
    Origin Offset=<0 -15 0>
    Override Seed=false
    Velocity Mode=::NewEnumerator1
    Velocity Speed=&RandomRangeFloat(Minimum=200, Maximum=300)
   ) {enabled=false, module=AddVelocity002}
   &SkeletalMeshLocation(Bone Sampling Mode=::NewEnumerator4, MaskSpawnGroup=false, Skeletal Mesh=$User.SkeletalMesh) {enabled=false}
   &ApplyInitialForces() {enabled=false}
  ]
  particleUpdateScript [
   &ParticleState()
   &AccelerationForce() {enabled=false}
   &Drag(Drag=8, Use Linear Drag=true, Use Rotational Drag=false)
   &CurlNoiseForce(
    Noise Frequency=0.1
    Noise Strength=&FloatFromCurve(FloatCurve=DataInterface() {propertyValues {Curve {keys [<at 0 1>,<at 1 0>]}}}, Scale Curve=1000)
    Pan Noise=true
    Pan Noise Field=<0 0 1>
   ) {enabled=false}
   &AccelerationForce(Acceleration=<0 0 250>) {enabled=false, module=AccelerationForce001}
   &VortexForce(Origin Pull Amount=40, UseOriginOffset=false, Vortex Force Amount=800) {enabled=false}
   &PointAttractionForce(Attraction Radius=400, Attraction Strength=200, Kill Within Radius=false, Module.Use Falloff=true, UseAttractorOffset=false)
   /SetVariables(Particles.Position=/Niagara/DynamicInputs/Lerp/SmoothLerpOverTime_Position(
     Convergence Rate:NiagaraFloat=&FloatFromCurve(
      FloatCurve=DataInterface() {propertyValues {Curve {keys [<at 0 0>,<at 0.6 1>]}}}
      Scale Curve=&FloatFromCurve(FloatCurve=DataInterface() {propertyValues {Curve {keys [<at 0 4>,<at 1 14>]}}}, CurveIndex=$Emitter.NormalizedLoopAge)
     )
     Smooth Value:NiagaraPosition=$Particles.Position
     Target Value:NiagaraPosition=<40 0 0>
    )) {isSetParameters=true, module=SetVariables_#3}
   &SolveForcesAndVelocity(Clamp Velocity=false, Limit Acceleration=false)
   &ScaleColor(
    Scale Alpha=&FloatFromCurve(FloatCurve=DataInterface() {propertyValues {Curve {keys [
         {arriveTangent=4.1785, interpMode=RCIM_Cubic, leaveTangent=4.1785, tangentMode=RCTM_User, time=0, value=0}
         {arriveTangent=0.0224, interpMode=RCIM_Cubic, leaveTangent=0.0224, tangentMode=RCTM_User, time=1, value=1}
        ]}}})
    Scale RGB=&MakeVectorFromLinearColorRGB(Color=&HueShiftLinearColor(
      Hue=0.25
      Saturation=1
      Value=$System.IntensityOverTime
      Linear Color=&ColorFromCurve(ColorCurve=DataInterface() {propertyValues {
         AlphaCurve {keys [<at 0 1 RCIM_Cubic>]}
         BlueCurve {keys [<at 0.0265 0>,<at 0.1089 0 RCIM_Cubic>,<at 1 0 RCIM_Cubic>]}
         GreenCurve {keys [<at 0.0265 0>,<at 0.1089 1.8333 RCIM_Cubic>,<at 1 0 RCIM_Cubic>]}
         RedCurve {keys [<at 0.0265 0.2>, {arriveTangent=0.8, interpMode=RCIM_Cubic, time=0.1089, value=10}, <at 1 1 RCIM_Cubic>]}
        }})
     ))
    ScaleA=true
    ScaleRGB=true
   )
   &ScaleSpriteSizeBySpeed(
    Max Scale Factor=<1.666 13>
    Min Scale Factor=<1.333 1.333>
    Sample Scale Factor By Curve=true
    Scale Factor Curve=DataInterface() {propertyValues {Curve {keys [<at 0 0>, {arriveTangent=1.4989, interpMode=RCIM_Cubic, tangentMode=RCTM_Break, time=1, value=1}]}}}
    Velocity Threshold=999
   )
   &ScaleSpriteSize(
    Uniform Curve Scale=&Multiply_Float(A=0.5, B=1)
    Uniform Curve Sprite Scale=DataInterface() {propertyValues {Curve {keys [
        {tangentMode=RCTM_User, time=0, value=1}
        {arriveTangent=-0.0405, interpMode=RCIM_Cubic, leaveTangent=-0.0405, tangentMode=RCTM_User, time=0.7009, value=1}
        {arriveTangent=-8.0089, interpMode=RCIM_Cubic, leaveTangent=-8.0089, tangentMode=RCTM_User, time=1, value=0}
       ]}}}
   )
  ]
  rendererClasses [&NiagaraSpriteRendererProperties]
  renderers [{rendererClass=&NiagaraSpriteRendererProperties}]
 }
 Cylinder_Mesh {
  emitterUpdateScript [
   &EmitterState()
   &SpawnBurst_Instantaneous(Use Spawn Probability=false, UseLoopCountLimit=false)
  ]
  particleSpawnScript [
   &InitializeParticle(
    Color=<rgba 4 0 1.5333 40>
    Lifetime=5.5
    Mesh Scale=<0.1 0.1 0.2>
    Mesh Scale Mode=::NewEnumerator4
    Position Offset=<30 0 0>
    UsePositionOffset=true
    Write Material Random=true
   )
   &InitialMeshOrientation(Mesh Axis To Orient=<0 0 1>, Mesh Orientation Mode=::NewEnumerator2, Orientation Coordinate Space=::Local, Use Rotation Vector=false)
   &CameraOffset(Camera Offset Amount=-4)
  ]
  particleUpdateScript [
   &ParticleState()
   &ScaleColor(
    Scale Alpha=&FloatFromCurve(
     FloatCurve=DataInterface() {propertyValues {Curve {keys [
         {interpMode=RCIM_Cubic, tangentMode=RCTM_User, time=0.2, value=0}
         {interpMode=RCIM_Cubic, tangentMode=RCTM_User, time=0.5, value=1}
         {tangentMode=RCTM_User, time=0.9, value=0}
        ]}}}
     Scale Curve=0.25
    )
    ScaleA=true
    ScaleRGB=true
   )
   &ScaleMeshSize(Scale Factor=&MakeVector(
     X=1
     Y=1
     Z=&FloatFromCurve(
      FloatCurve=DataInterface() {propertyValues {Curve {keys [<at 0.3 0.8>, {arriveTangent=-1.1429, interpMode=RCIM_Cubic, leaveTangent=-1.1429, time=0.8, value=1.5}, <at 1 0>]}}}
      CurveIndex=$System.NormalizedLoopAge
      Scale Curve=0.5
     )
    )) {module=ScaleMeshSize002}
   &ScaleColor(Scale Alpha=$System.IntensityOverTime, ScaleA=true, ScaleRGB=true) {module=ScaleColor001}
  ]
  rendererClasses [&NiagaraMeshRendererProperties]
  renderers [{rendererClass=&NiagaraMeshRendererProperties, rendererIndex=0}]
 }
 Debris {
  emitterUpdateScript [
   &EmitterState(Life Cycle Mode=::NewEnumerator1, Loop Behavior=::NewEnumerator1, Loop Duration=3.2, UseLoopDelay=true)
   /SetVariables(Emitter.BuildupCurve:NiagaraFloat=&FloatFromCurve(
     FloatCurve=DataInterface() {propertyValues {Curve {keys [
         {arriveTangent=0.0145, interpMode=RCIM_Cubic, leaveTangent=0.0145, tangentMode=RCTM_User, time=0, value=0}
         {arriveTangent=2.5625, interpMode=RCIM_Cubic, leaveTangent=2.5625, tangentMode=RCTM_User, time=0.8, value=1}
         {interpMode=RCIM_Cubic, leaveTangent=1.25, time=0.8373, value=0}
        ]}}}
     CurveIndex=$Emitter.NormalizedLoopAge
    )) {isSetParameters=true}
   /SetVariables(Emitter.BuildupCurveInfinite:NiagaraFloat=&FloatFromCurve(
     FloatCurve=DataInterface() {propertyValues {Curve {keys [
         {arriveTangent=0.0145, interpMode=RCIM_Cubic, leaveTangent=0.0145, tangentMode=RCTM_User, time=0, value=0}
         {arriveTangent=2.5625, interpMode=RCIM_Cubic, leaveTangent=2.5625, tangentMode=RCTM_User, time=1, value=1}
        ]}}}
     CurveIndex=$Emitter.NormalizedLoopAge
    )) {isSetParameters=true}
   &SpawnRate(SpawnRate=&Multiply_Float(A=$Emitter.BuildupCurve, B=40), Use Spawn Probability=false) {module=SpawnRate001}
  ]
  particleSpawnScript [
   &InitializeParticle(
    Color=&MultiplyLinearColor(
     Linear Color=&RandomRangeLinearColor(Minimum=<rgba 0.633 0.153 1 1>, Maximum=<rgba 0.4876 0.146 1 1>)
     Scale Factor=&MakeVector4(X=1, Y=1, Z=1, W=&Multiply_Float(A=$System.BuildupCurve, B=8))
    )
    Lifetime Max=1.8
    Lifetime Mode=::NewEnumerator1
    Mass Mode=::NewEnumerator3
    Sprite Rotation Mode=::NewEnumerator0
    Sprite Size Mode=::NewEnumerator1
    Uniform Sprite Size Max=3
    Uniform Sprite Size Min=1.5
    UsePositionOffset=false
    Write Material Random=true
   )
   &ShapeLocation(
    MaskSpawnGroup=false
    Offset=<0 -22 0>
    Offset Coordinate Space=::Local
    Offset Mode=::NewEnumerator0
    Override Seed=false
    Sphere Radius=5
    UseOwnerScale=true
   ) {enabled=false}
   &ShapeLocation(
    MaskSpawnGroup=false
    Offset=&MakeVector(
     X=&FloatFromCurve(
      FloatCurve=DataInterface() {propertyValues {Curve {keys [{tangentMode=RCTM_User, time=0.1, value=0.7}, {tangentMode=RCTM_User, time=1, value=1}]}}}
      CurveIndex=$Emitter.BuildupCurve
      Scale Curve=200
     )
     Y=0
     Z=0
    )
    Offset Mode=::NewEnumerator0
    Override Seed=false
    Rotation Mode=::NewEnumerator2
    Sphere Radius=&FloatFromCurve(
     FloatCurve=DataInterface() {propertyValues {Curve {keys [{tangentMode=RCTM_User, time=0, value=0.1}, {tangentMode=RCTM_User, time=1, value=1}]}}}
     CurveIndex=$Emitter.BuildupCurveInfinite
     Scale Curve=350
    )
    UseOwnerScale=true
   ) {module=ShapeLocation001}
   &AddVelocity(Velocity=<0 10 0>)
  ]
  particleUpdateScript [
   &ParticleState()
   /SetVariables(Particles.Position=&LerpPosition(A=$Particles.Position, B=<30 0 0>, Alpha=$Particles.NormalizedAge)) {isSetParameters=true}
   &ScaleSpriteSize(
    Scale Sprite Size Mode=::NewEnumerator0
    Uniform Scale Factor=&FloatFromCurve(
     FloatCurve=DataInterface() {propertyValues {Curve {keys [
         {arriveTangent=9.0725, interpMode=RCIM_Cubic, leaveTangent=9.0725, tangentMode=RCTM_User, time=0, value=0}
         {arriveTangent=-2.5249, interpMode=RCIM_Cubic, leaveTangent=-2.5249, tangentMode=RCTM_User, time=0.1408, value=0.9847}
         {arriveTangent=-0.0405, interpMode=RCIM_Cubic, leaveTangent=-0.0405, tangentMode=RCTM_User, time=1, value=0.3}
        ]}}}
     Scale Curve=2
    )
   ) {module=ScaleSpriteSize001}
   &ScaleColor(
    Scale Alpha=&FloatFromCurve(
     FloatCurve=DataInterface() {propertyValues {Curve {keys [
         {arriveTangent=9.0725, interpMode=RCIM_Cubic, leaveTangent=9.0725, tangentMode=RCTM_User, time=0, value=0}
         {arriveTangent=-2.5249, interpMode=RCIM_Cubic, leaveTangent=-2.5249, tangentMode=RCTM_User, time=0.1408, value=0.9847}
         {arriveTangent=-0.0405, interpMode=RCIM_Cubic, leaveTangent=-0.0405, tangentMode=RCTM_User, time=1, value=0}
        ]}}}
     Scale Curve=0.2
    )
    Scale RGB=&VectorFromFloat(Value=1)
    ScaleA=true
    ScaleRGB=true
   )
   &ScaleVelocity(Velocity Scale=&VectorFromFloat(Value=&FloatFromCurve(FloatCurve=DataInterface() {propertyValues {Curve {keys [{tangentMode=RCTM_User, time=0.5, value=0.5}, {tangentMode=RCTM_User, time=1, value=1}]}}}))) {enabled=false}
   /Niagara/Modules/Update/Forces/LineAttractionForce(Line End:Vector3f=<0 -100 0>, Line Start:Vector3f=<0 100 0>, Use Attraction Falloff=false) {enabled=false}
   &SolveForcesAndVelocity(Clamp Velocity=false, Limit Acceleration=false)
   &DynamicMaterialParameters(Write Parameter Index 0=false)
   &SpriteFacingAndAlignment(
    Sprite Alignment=&NormalizeVector(
     Vector To Normalize=&AddVectorToPosition(
      Position=Linked() {linkedVariable {name=Particles.Velocity, type=Vector3f}}
      Vector=&SubtractVectorFromPosition(Position=$Particles.Position, Vector=Linked() {linkedVariable {name=Particles.Previous.Position, type=NiagaraPosition}})
     )
     Safe Normalize=true
     Safe Fallback Threshold=0
     Safe Fallback Vector=<0 0 0.001>
    )
    WriteAlignment=true
    WriteFacing=false
   )
  ]
  rendererClasses [&NiagaraSpriteRendererProperties]
  renderers [
   {rendererClass=&NiagaraSpriteRendererProperties}
   {rendererClass=&NiagaraSpriteRendererProperties, rendererIndex=1}
  ]
 }
 DysfunctionSpark {
  emitterUpdateScript [
   &EmitterState(Life Cycle Mode=::NewEnumerator1, Loop Behavior=::NewEnumerator1, Loop Delay=3.4, Loop Duration=2.5, UseLoopDelay=true)
   &SpawnRate(SpawnRate=200, Use Spawn Probability=false) {enabled=false, module=SpawnRate001}
   &SpawnBurst_Instantaneous(Spawn Count=8, Spawn Time=0.2, Use Spawn Probability=false, UseLoopCountLimit=false)
   &SpawnRate(
    SpawnRate=&FloatFromCurve(
     FloatCurve=DataInterface() {propertyValues {Curve {keys [<at 0 0>,<at 0.7 1>,<at 1 0>]}}}
     CurveIndex=$Emitter.NormalizedLoopAge
     Scale Curve=60
    )
    Use Spawn Probability=false
   ) {enabled=false}
  ]
  particleSpawnScript [
   /SetVariables(Particles.PerParticleRandom:NiagaraFloat=&RandomRangeFloat(Minimum=0, Maximum=1)) {isSetParameters=true, module=SetVariables_#2}
   &InitializeParticle(
    Color Mode=::NewEnumerator0
    Lifetime=&Lerp_Float(A=1.5, B=1.2, Alpha=$Particles.PerParticleRandom)
    Sprite Size=&Multiply_Vector2D(
     A=<1.2 1.2>
     B=&MakeVector2D(
      X=&FloatFromCurve(
       FloatCurve=DataInterface() {propertyValues {Curve {keys [<at 0 1>,<at 0.8 2 RCIM_Constant>,<at 0.95 3 RCIM_Constant>]}}}
       CurveIndex=$Particles.PerParticleRandom
      )
      Y=&FloatFromCurve(
       FloatCurve=DataInterface() {propertyValues {Curve {keys [<at 0 1>,<at 0.8 2 RCIM_Constant>,<at 0.95 3>]}}}
       CurveIndex=$Particles.PerParticleRandom
      )
     )
    )
    Sprite Size Mode=::NewEnumerator4
    UsePositionOffset=false
    Write Material Random=true
   )
   &SkeletalMeshLocation(Bone Sampling Mode=::NewEnumerator0, MaskSpawnGroup=false, OverrideRandomSeed=false, Skeletal Mesh=$User.SkeletalMesh) {enabled=false, module=SkeletalMeshLocation002}
   &AddVelocity(
    Cone Angle=33
    Cone Axis=<0 0.25 1>
    Module.UseConeFalloff=false
    Module.UseVelDistribution=false
    Override Seed=false
    Velocity Mode=::NewEnumerator2
    Velocity Speed=&FloatFromCurve(
     FloatCurve=DataInterface() {propertyValues {Curve {keys [<at 0 0.5>,<at 0.333 1>,<at 0.666 0.1>,<at 3 0>]}}}
     CurveIndex=$Emitter.LoopedAge
     Scale Curve=&RandomRangeFloat(Minimum=444, Maximum=999)
    )
   ) {enabled=false}
   &ShapeLocation(
    MaskSpawnGroup=false
    Offset=<30 0 0>
    Offset Coordinate Space=::Local
    Offset Mode=::NewEnumerator0
    Override Seed=false
    Rotation Mode=::NewEnumerator2
    Sphere Radius=1
    UseOwnerScale=true
   )
   &InheritVelocity(Inherited Velocity Amount Scale=&VectorFromFloat(Value=3), Speed Limit=true, UseSpeedThreshold=false) {enabled=false}
   &InheritVelocity(Speed Limit=true, UseSpeedThreshold=false) {enabled=false, module=InheritVelocity001}
   &AddVelocity(
    Cone Angle=80
    Cone Axis=<-0.3 1 0>
    Module.UseConeFalloff=false
    Module.UseVelDistribution=false
    Override Seed=false
    Rotation Angle Type:ENiagara_AngleInput=::NewEnumerator1
    Rotation Mode=::NewEnumerator2
    Velocity Mode=::NewEnumerator2
    Velocity Speed=&Multiply_Float(A=1.2, B=&RandomRangeFloat(Minimum=100, Maximum=700))
    Yaw / Pitch / Roll:Vector3f=<-0.3 0 0>
   ) {module=AddVelocity001}
   &AddVelocity(
    Module.Use Falloff Distance=false
    Origin Offset=<0 -15 0>
    Override Seed=false
    Velocity Mode=::NewEnumerator1
    Velocity Speed=&RandomRangeFloat(Minimum=200, Maximum=300)
   ) {enabled=false, module=AddVelocity002}
   &SkeletalMeshLocation(Bone Sampling Mode=::NewEnumerator4, MaskSpawnGroup=false, Skeletal Mesh=$User.SkeletalMesh) {enabled=false}
  ]
  particleUpdateScript [
   &ParticleState()
   &AccelerationForce() {enabled=false}
   &Drag(Drag=9, Use Linear Drag=true, Use Rotational Drag=false)
   &CurlNoiseForce(Noise Frequency=0.1, Noise Strength=1000, Pan Noise=true, Pan Noise Field=<0 0 1>)
   &AccelerationForce(Acceleration=<0 0 250>) {module=AccelerationForce001}
   &GravityForce(Gravity=&MakeVector(
     X=0
     Y=0
     Z=&FloatFromCurve(
      FloatCurve=DataInterface() {propertyValues {Curve {keys [
          {interpMode=RCIM_Cubic, tangentMode=RCTM_User, time=0, value=0}
          {arriveTangent=1.5796, interpMode=RCIM_Cubic, leaveTangent=1.5796, tangentMode=RCTM_User, time=1, value=1}
         ]}}}
      Scale Curve=980
     )
    ))
   &InheritSourceMovement(
    Applied Position Scale=&VectorFromFloat(Value=&FloatFromCurve(FloatCurve=DataInterface() {propertyValues {Curve {keys [<at 0.2 1>,<at 0.5 0>]}}}))
    ApplyForce=false
    ApplyPosition=true
    ApplyPositionFalloff=false
   )
   &SolveForcesAndVelocity(Clamp Velocity=false, Limit Acceleration=false)
   &ScaleColor(
    Scale Alpha=2
    Scale RGB=&MakeVectorFromLinearColorRGB(Color=&HueShiftLinearColor(
      Hue=0.25
      Saturation=1
      Value=1
      Linear Color=&ColorFromCurve(ColorCurve=DataInterface() {propertyValues {
         AlphaCurve {keys [<at 0 1 RCIM_Cubic>]}
         BlueCurve {keys [<at 0.0265 0>,<at 0.1089 0 RCIM_Cubic>,<at 1 0 RCIM_Cubic>]}
         GreenCurve {keys [<at 0.0265 0>,<at 0.1089 1.8333 RCIM_Cubic>,<at 1 0 RCIM_Cubic>]}
         RedCurve {keys [<at 0.0265 0.2>, {arriveTangent=0.8, interpMode=RCIM_Cubic, time=0.1089, value=10}, <at 1 1 RCIM_Cubic>]}
        }})
     ))
    ScaleA=true
    ScaleRGB=true
   )
   &ScaleSpriteSizeBySpeed(
    Max Scale Factor=<1.666 13>
    Min Scale Factor=<1.333 1.333>
    Sample Scale Factor By Curve=true
    Scale Factor Curve=DataInterface() {propertyValues {Curve {keys [<at 0 0>, {arriveTangent=1.4989, interpMode=RCIM_Cubic, tangentMode=RCTM_Break, time=1, value=1}]}}}
    Velocity Threshold=999
   )
   &ScaleSpriteSize(
    Uniform Curve Scale=&Multiply_Float(A=0.75, B=&RandomRangeFloat(Minimum=0.7, Maximum=1.3))
    Uniform Curve Sprite Scale=DataInterface() {propertyValues {Curve {keys [
        {tangentMode=RCTM_User, time=0, value=1}
        {arriveTangent=-0.0405, interpMode=RCIM_Cubic, leaveTangent=-0.0405, tangentMode=RCTM_User, time=0.7009, value=1}
        {arriveTangent=-8.0089, interpMode=RCIM_Cubic, leaveTangent=-8.0089, tangentMode=RCTM_User, time=1, value=0}
       ]}}}
   )
  ]
  rendererClasses [&NiagaraSpriteRendererProperties]
  renderers [{rendererClass=&NiagaraSpriteRendererProperties}]
 }
 DysfunctionSpark001 {
  emitterUpdateScript [
   &EmitterState(Life Cycle Mode=::NewEnumerator1, Loop Behavior=::NewEnumerator1, Loop Delay=3.4, Loop Duration=2.5, UseLoopDelay=true)
   &SpawnRate(SpawnRate=200, Use Spawn Probability=false) {enabled=false, module=SpawnRate001}
   &SpawnBurst_Instantaneous(Spawn Count=5, Spawn Time=0.2, Use Spawn Probability=false, UseLoopCountLimit=false)
   &SpawnRate(
    SpawnRate=&FloatFromCurve(
     FloatCurve=DataInterface() {propertyValues {Curve {keys [<at 0 0>,<at 0.7 1>,<at 1 0>]}}}
     CurveIndex=$Emitter.NormalizedLoopAge
     Scale Curve=60
    )
    Use Spawn Probability=false
   ) {enabled=false}
  ]
  particleSpawnScript [
   /SetVariables(Particles.PerParticleRandom:NiagaraFloat=&RandomRangeFloat(Minimum=0, Maximum=1)) {isSetParameters=true, module=SetVariables_#2}
   &InitializeParticle(
    Color Mode=::NewEnumerator0
    Lifetime=&Lerp_Float(A=0.6, B=1.5, Alpha=$Particles.PerParticleRandom)
    Sprite Size Mode=::NewEnumerator3
    Uniform Sprite Size=5
    UsePositionOffset=false
    Write Material Random=true
   )
   &SkeletalMeshLocation(Bone Sampling Mode=::NewEnumerator0, MaskSpawnGroup=false, OverrideRandomSeed=false, Skeletal Mesh=$User.SkeletalMesh) {enabled=false, module=SkeletalMeshLocation002}
   &AddVelocity(
    Cone Angle=33
    Cone Axis=<0 0.25 1>
    Module.UseConeFalloff=false
    Module.UseVelDistribution=false
    Override Seed=false
    Velocity Mode=::NewEnumerator2
    Velocity Speed=&FloatFromCurve(
     FloatCurve=DataInterface() {propertyValues {Curve {keys [<at 0 0.5>,<at 0.333 1>,<at 0.666 0.1>,<at 3 0>]}}}
     CurveIndex=$Emitter.LoopedAge
     Scale Curve=&RandomRangeFloat(Minimum=444, Maximum=999)
    )
   ) {enabled=false}
   &ShapeLocation(
    MaskSpawnGroup=false
    Offset=<30 0 0>
    Offset Coordinate Space=::Local
    Offset Mode=::NewEnumerator0
    Override Seed=false
    Rotation Mode=::NewEnumerator2
    Sphere Radius=1
    UseOwnerScale=true
   )
   &InheritVelocity(Inherited Velocity Amount Scale=&VectorFromFloat(Value=3), Speed Limit=true, UseSpeedThreshold=false) {enabled=false}
   &InheritVelocity(Speed Limit=true, UseSpeedThreshold=false) {enabled=false, module=InheritVelocity001}
   &AddVelocity(
    Cone Angle=80
    Cone Axis=<-0.3 1 0>
    Module.UseConeFalloff=false
    Module.UseVelDistribution=false
    Override Seed=false
    Rotation Angle Type:ENiagara_AngleInput=::NewEnumerator1
    Rotation Mode=::NewEnumerator2
    Velocity Mode=::NewEnumerator2
    Velocity Speed=&Multiply_Float(A=1.2, B=&RandomRangeFloat(Minimum=100, Maximum=700))
    Yaw / Pitch / Roll:Vector3f=<-0.3 0 0>
   ) {module=AddVelocity001}
   &AddVelocity(
    Module.Use Falloff Distance=false
    Origin Offset=<0 -15 0>
    Override Seed=false
    Velocity Mode=::NewEnumerator1
    Velocity Speed=&RandomRangeFloat(Minimum=200, Maximum=300)
   ) {enabled=false, module=AddVelocity002}
   &SkeletalMeshLocation(Bone Sampling Mode=::NewEnumerator4, MaskSpawnGroup=false, Skeletal Mesh=$User.SkeletalMesh) {enabled=false}
  ]
  particleUpdateScript [
   &ParticleState()
   &AccelerationForce() {enabled=false}
   &Drag(Drag=15, Use Linear Drag=true, Use Rotational Drag=false)
   &CurlNoiseForce(Noise Frequency=0.1, Noise Strength=1000, Pan Noise=true, Pan Noise Field=<0 0 1>)
   &AccelerationForce(Acceleration=<0 0 250>) {module=AccelerationForce001}
   &GravityForce(Gravity=&MakeVector(
     X=0
     Y=0
     Z=&FloatFromCurve(
      FloatCurve=DataInterface() {propertyValues {Curve {keys [
          {interpMode=RCIM_Cubic, tangentMode=RCTM_User, time=0, value=0}
          {arriveTangent=1.5796, interpMode=RCIM_Cubic, leaveTangent=1.5796, tangentMode=RCTM_User, time=1, value=1}
         ]}}}
      Scale Curve=980
     )
    ))
   &InheritSourceMovement(
    Applied Position Scale=&VectorFromFloat(Value=&FloatFromCurve(FloatCurve=DataInterface() {propertyValues {Curve {keys [<at 0.2 1>,<at 0.5 0>]}}}))
    ApplyForce=false
    ApplyPosition=true
    ApplyPositionFalloff=false
   )
   &SolveForcesAndVelocity(Clamp Velocity=false, Limit Acceleration=false)
   &ScaleColor(
    Scale Alpha=0.02
    Scale RGB=&MakeVectorFromLinearColorRGB(Color=&HueShiftLinearColor(
      Hue=0.25
      Saturation=1
      Value=1
      Linear Color=&ColorFromCurve(ColorCurve=DataInterface() {propertyValues {
         AlphaCurve {keys [<at 0 1 RCIM_Cubic>]}
         BlueCurve {keys [<at 0.0265 0>,<at 0.1089 0 RCIM_Cubic>,<at 1 0 RCIM_Cubic>]}
         GreenCurve {keys [<at 0.0265 0>,<at 0.1089 1.8333 RCIM_Cubic>,<at 1 0 RCIM_Cubic>]}
         RedCurve {keys [<at 0.0265 0.2>, {arriveTangent=0.8, interpMode=RCIM_Cubic, time=0.1089, value=10}, <at 1 1 RCIM_Cubic>]}
        }})
     ))
    ScaleA=true
    ScaleRGB=true
   )
   &ScaleSpriteSizeBySpeed(
    Max Scale Factor=<1.666 13>
    Min Scale Factor=<1.333 1.333>
    Sample Scale Factor By Curve=true
    Scale Factor Curve=DataInterface() {propertyValues {Curve {keys [<at 0 0>, {arriveTangent=1.4989, interpMode=RCIM_Cubic, tangentMode=RCTM_Break, time=1, value=1}]}}}
    Velocity Threshold=999
   ) {enabled=false}
   &ScaleSpriteSize(
    Uniform Curve Scale=&Multiply_Float(A=0.75, B=&RandomRangeFloat(Minimum=0.7, Maximum=1.3))
    Uniform Curve Sprite Scale=DataInterface() {propertyValues {Curve {keys [
        {tangentMode=RCTM_User, time=0, value=1}
        {arriveTangent=-0.0405, interpMode=RCIM_Cubic, leaveTangent=-0.0405, tangentMode=RCTM_User, time=0.7009, value=1}
        {arriveTangent=-8.0089, interpMode=RCIM_Cubic, leaveTangent=-8.0089, tangentMode=RCTM_User, time=1, value=0}
       ]}}}
   )
  ]
  rendererClasses [&NiagaraSpriteRendererProperties]
  renderers [{rendererClass=&NiagaraSpriteRendererProperties}]
 }
 FN_Debug {
  bEnabled=false
  emitterUpdateScript [
   &EmitterState()
   &SpawnBurst_Instantaneous(Spawn Count=1, Use Spawn Probability=false)
  ]
  particleSpawnScript [
   /Niagara/Modules/Spawn/Initialization/InitializeParticle(
    Lifetime=5
    Module.Write Color=true
    Module.Write Lifetime=true
    Module.Write Mass=true
    Module.Write Material Random=true
    Module.Write Position=true
    Module.Write Scale=false
    Module.Write SpriteRotation=false
    Module.Write SpriteSize=true
    Sprite Size=<100 100>
   )
   /Niagara/Modules/Spawn/Initialization/V2/InitializeParticle(
    Mesh Scale Mode=::NewEnumerator3
    Mesh Uniform Scale=2
    Position Offset Coordinate Space=::Local
    UsePositionOffset=true
    Write Material Random=true
   ) {enabled=false, module=InitializeParticle001}
  ]
  particleUpdateScript [
   &ParticleState()
   /SetVariables(Particles.Debug:NiagaraFloat=&DotProduct(
     Vector A=&FN_GetSystemAxis(Axis:ENiagaraOrientationAxis=::XAxis, Flip Axis:NiagaraBool=false)
     Vector B=&ReturnCameraVector(Camera Vectors=::NewEnumerator1, Coordinate Space=::Local)
    )) {isSetParameters=true}
   &DynamicMaterialParameters(
    Index 0 Param 1=$Particles.Debug
    Module.Param0WriteEnabled=true
    Module.Param1WriteEnabled=true
    Module.Param2WriteEnabled=true
    Module.Param3WriteEnabled=true
   )
  ]
  rendererClasses [&NiagaraSpriteRendererProperties, &NiagaraMeshRendererProperties]
  renderers [
   {rendererClass=&NiagaraSpriteRendererProperties}
   {rendererClass=&NiagaraMeshRendererProperties, rendererIndex=1}
  ]
 }
 GlowCard {
  emitterUpdateScript [
   &EmitterState(
    Inactive Response=::NewEnumerator1
    Life Cycle Mode=::NewEnumerator1
    Loop Behavior=::NewEnumerator1
    Loop Duration Mode=::NewEnumerator1
    UseLoopDelay=false
   )
   &SpawnBurst_Instantaneous(Spawn Time=0.4, Use Spawn Probability=false, UseLoopCountLimit=false)
  ]
  particleSpawnScript [
   &FN_SetSpawnGroup(Number of Spawn Groups=1)
   &InitializeParticle(
    Color=<rgba 83.25 0 333 1>
    Position Offset=<36.5 0 0>
    Sprite Size Mode=::NewEnumerator3
    Uniform Sprite Size=&Lerp_Float(
     A=6
     B=&Multiply_Float(A=11, B=&MakeFloatFromVector(Vector=$Engine.Owner.Scale, Channel=::NewEnumerator0))
     Alpha=0
    )
    UsePositionOffset=true
    Write Material Random=true
   )
   &SpriteFacingAndAlignment(Facing Coordinate Space=::Local, WriteAlignment=false, WriteFacing=true)
  ]
  particleUpdateScript [
   &ParticleState(Kill Particles When Lifetime Has Elapsed=false)
   &ScaleColor(
    Scale Alpha=$System.IntensityOverTime
    Scale RGB=&VectorFromFloat(Value=&Multiply_Float(A=200, B=$System.IntensityOverTime))
    ScaleA=true
    ScaleRGB=true
   ) {enabled=false}
   &ScaleColor(
    Scale Alpha=&FloatFromCurve(FloatCurve=DataInterface() {propertyValues {Curve {keys [<at 0.5 1>,<at 0.75 0>]}}}, CurveIndex=$System.NormalizedLoopAge)
    ScaleA=true
    ScaleRGB=true
   ) {module=ScaleColor002}
   &ScaleColor(
    Scale Alpha=&FloatFromCurve(FloatCurve=DataInterface() {propertyValues {Curve {keys [
         {arriveTangent=4.1785, interpMode=RCIM_Cubic, tangentMode=RCTM_User, time=0, value=0}
         {interpMode=RCIM_Cubic, leaveTangent=0.0224, tangentMode=RCTM_User, time=1, value=1}
        ]}}})
    ScaleA=true
    ScaleRGB=true
   ) {module=ScaleColor001}
  ]
  rendererClasses [&NiagaraSpriteRendererProperties]
  renderers [{rendererClass=&NiagaraSpriteRendererProperties}]
 }
 GlowFlash {
  bEnabled=false
  emitterUpdateScript [
   &EmitterState(
    Inactive Response=::NewEnumerator1
    Life Cycle Mode=::NewEnumerator1
    Loop Behavior=::NewEnumerator1
    Loop Delay=3.42
    Loop Duration=1
    UseLoopDelay=true
   )
   &SpawnBurst_Instantaneous(Use Spawn Probability=false, UseLoopCountLimit=false)
  ]
  particleSpawnScript [
   &FN_SetSpawnGroup(Number of Spawn Groups=1)
   &InitializeParticle(
    Color=<rgba 3.0987 0.84 7 0.4>
    Lifetime=0.05
    Position Offset=<35 0 0>
    Sprite Size Mode=::NewEnumerator3
    Uniform Sprite Size=&Multiply_Float(A=50, B=&MakeFloatFromVector(Vector=$Engine.Owner.Scale, Channel=::NewEnumerator0))
    UsePositionOffset=true
    Write Material Random=true
   )
   &CameraOffset(Camera Offset Amount=40)
  ]
  particleUpdateScript [
   &ParticleState(Kill Particles When Lifetime Has Elapsed=false)
   &ScaleColor(
    Scale Alpha=&FloatFromCurve(
     FloatCurve=DataInterface() {propertyValues {Curve {keys [<at 0 1>,<at 2 0>]}}}
     Scale Curve=&FNTimeOfDayScalar(Night Scale=0.1, Morning Scale=0.2, Day Scale=0.2, Evening Scale=0.1)
    )
    ScaleA=true
    ScaleRGB=false
   )
   /SetVariables(
    Particles.FlareCol:LinearColor=&MultiplyLinearColor(Linear Color=$Particles.Color, Scale Factor=&MakeVector4_FromFloat(Float=4))
    Particles.FlareSize:Vector2f=&Multiply_Vector2D_ByFloat(Float=2, Vector2D=$Particles.SpriteSize)
   ) {isSetParameters=true}
  ]
  rendererClasses [&NiagaraSpriteRendererProperties]
  renderers [
   {rendererClass=&NiagaraSpriteRendererProperties}
   {rendererClass=&NiagaraSpriteRendererProperties, rendererIndex=1}
   {rendererClass=&NiagaraSpriteRendererProperties, rendererIndex=2}
   {rendererClass=&NiagaraSpriteRendererProperties, rendererIndex=3}
  ]
 }
 GunGlow {
  emitterUpdateScript [
   &EmitterState(
    Inactive Response=::NewEnumerator1
    Life Cycle Mode=::NewEnumerator1
    Loop Behavior=::NewEnumerator1
    Loop Duration Mode=::NewEnumerator1
    UseLoopDelay=false
   )
   &SpawnBurst_Instantaneous(Use Spawn Probability=false, UseLoopCountLimit=false)
  ]
  particleSpawnScript [&InitializeParticle(
    Color=<rgba 1.184 0.48 4 1>
    Lifetime=2.5
    Sprite Size Mode=::NewEnumerator3
    Uniform Sprite Size=55
    UsePositionOffset=true
    Write Material Random=true
   )]
  particleUpdateScript [
   &ParticleState(Kill Particles When Lifetime Has Elapsed=false)
   &ScaleColor(
    Scale Alpha=&Multiply_Float(
     A=$System.IntensityOverTime
     B=&FloatFromCurve(FloatCurve=DataInterface() {propertyValues {Curve {keys [<at 0 1>,<at 1 0>]}}}, Scale Curve=10)
    )
    Scale RGB=&VectorFromFloat(Value=1)
    ScaleA=true
    ScaleRGB=true
   )
   &DynamicMaterialParameters(
    Index 0 Param 1=&MakeFloatFromVector(Vector=$Engine.Owner.SystemXAxis, Channel=::NewEnumerator0)
    Index 0 Param 2=&MakeFloatFromVector(Vector=$Engine.Owner.SystemXAxis, Channel=::NewEnumerator1)
    Index 0 Param 3=&MakeFloatFromVector(Vector=$Engine.Owner.SystemXAxis, Channel=::NewEnumerator2)
    Index 0 Param 4=&FloatFromCurve(
     FloatCurve=DataInterface() {propertyValues {Curve {keys [{interpMode=RCIM_Cubic, tangentMode=RCTM_User, time=0, value=0}, {interpMode=RCIM_Cubic, tangentMode=RCTM_User, time=0.2, value=1}]}}}
     Scale Curve=100
    )
    Module.Param0WriteEnabled=true
    Module.Param1WriteEnabled=true
    Module.Param2WriteEnabled=true
    Module.Param3WriteEnabled=true
   )
  ]
  rendererClasses [&NiagaraSpriteRendererProperties, &NiagaraLightRendererProperties]
  renderers [
   {rendererClass=&NiagaraSpriteRendererProperties}
   {rendererClass=&NiagaraLightRendererProperties, rendererIndex=1}
  ]
 }
 HeatCones {
  emitterUpdateScript [
   &EmitterState(
    Inactive Response=::NewEnumerator1
    Life Cycle Mode=::NewEnumerator1
    Loop Behavior=::NewEnumerator1
    Loop Duration Mode=::NewEnumerator1
    UseLoopDelay=true
   )
   &SpawnBurst_Instantaneous(Use Spawn Probability=false, UseLoopCountLimit=true)
  ]
  particleSpawnScript [
   &InitializeParticle(
    Lifetime=7
    Mesh Scale=&Multiply_VectorByFloat(Vector=<0.2 0.2 3>, Float=&FN_FrontEndFloatModifier(Gameplay=1, Front End=0.6))
    Mesh Scale Mode=::NewEnumerator4
    Position Offset=<20 0 0>
    Position Offset Coordinate Space=::Local
    UsePositionOffset=true
    Write Material Random=true
   )
   &InitialMeshOrientation(Mesh Orientation Mode=::NewEnumerator7, Rotation=<0 -0.25 0>, Use Rotation Vector=true)
   &DynamicMaterialParameters(
    Index 0 Param 1=$User.Variant
    Index 0 Param 2=&MaskFloatBySpawnGroup(Masked Float=-1, Passthrough Float=1, Spawn Group Mask=$Particles.PartitionParticles.Partition)
    Module.Param0WriteEnabled=true
    Module.Param1WriteEnabled=true
    Module.Param2WriteEnabled=false
    Module.Param3WriteEnabled=false
   )
   &CameraOffset(Camera Offset Amount=-20)
  ]
  particleUpdateScript [
   /SetVariables(Particles.Color=&FN_Select_Color_from_Time_Of_Day(Morning=<rgba 0.8 0.8 0.8 1>, Day=<rgba 1 1 1 1>, Evening=<rgba 0.8 0.8 0.8 1>, Night=<rgba 0.6 0.6 0.6 1>)) {isSetParameters=true}
   &ParticleState(
    Kill Particles When Lifetime Has Elapsed=false
    Let Infinitely Lived Particles Die When Emitter Deactivates=true
    Loop Particles Lifetime=true
   )
   /SetVariables(Particles.MeshOrientation=&MultiplyQuaternion(
     Quaternion A=$Particles.MeshOrientation
     Quaternion B=/BRCosmetics/Athena/Effects/Particles/Skydiving/NS_Contrail_Apprentice/OrientMestToVectorFromCustomRot(
      InitialRotation:Quat4f=<0 0 0 1>
      Look At Direction:Vector3f=<0 0 0>
      Reference Axis Vector:Vector3f=<0 0 0>
      RotationVector:Vector3f=&MakeVector(
       X=0
       Y=0
       Z=&FloatFromCurve(FloatCurve=DataInterface() {propertyValues {Curve {keys [<at 0 0>,<at 1 1 RCIM_Cubic>]}}})
      )
     )
    )) {enabled=false, isSetParameters=true}
   &ScaleMeshSize(Scale Factor=&MaskVectorBySpawnGroup(Masked Vector=<-1 1 1>, Passthrough Vector=<1 1 1>, Spawn Group Mask=$Particles.PartitionParticles.Partition)) {enabled=false, module=ScaleMeshSize001}
   &ScaleMeshSize(Scale Factor=&MakeVector(
     X=1
     Y=1
     Z=&Multiply_Float(
      A=&FloatFromCurve(FloatCurve=DataInterface() {propertyValues {Curve {keys [
           <at 0 1 RCIM_Cubic>
           {arriveTangent=-0.75, interpMode=RCIM_Cubic, leaveTangent=-0.75, time=0.1, value=0.65}
           {arriveTangent=1, interpMode=RCIM_Cubic, leaveTangent=1, time=0.2, value=0.85}
           {arriveTangent=1.25, interpMode=RCIM_Cubic, leaveTangent=1.25, time=0.3, value=0.85}
           {arriveTangent=0.25, interpMode=RCIM_Cubic, leaveTangent=0.25, time=0.4, value=1.1}
           {arriveTangent=-0.5, interpMode=RCIM_Cubic, leaveTangent=-0.5, time=0.5, value=0.9}
           {arriveTangent=-0.75, interpMode=RCIM_Cubic, leaveTangent=-0.75, time=0.6, value=1}
           {arriveTangent=-2, interpMode=RCIM_Cubic, leaveTangent=-2, time=0.7, value=0.75}
           {arriveTangent=1, interpMode=RCIM_Cubic, leaveTangent=1, time=0.8, value=0.6}
           {arriveTangent=2, interpMode=RCIM_Cubic, leaveTangent=2, time=0.9, value=0.95}
           <at 1 1 RCIM_Cubic>
          ]}}})
      B=&FloatFromCurve(
       FloatCurve=DataInterface() {propertyValues {Curve {keys [
           <at 0 1 RCIM_Cubic>
           {arriveTangent=-0.5, interpMode=RCIM_Cubic, leaveTangent=-0.5, time=0.1, value=0.85}
           {arriveTangent=0.5, interpMode=RCIM_Cubic, leaveTangent=0.5, time=0.2, value=0.9}
           <at 0.3 0.95 RCIM_Cubic>
           {arriveTangent=-1, interpMode=RCIM_Cubic, leaveTangent=-1, time=0.4, value=0.9}
           {arriveTangent=-0.25, interpMode=RCIM_Cubic, leaveTangent=-0.25, time=0.5, value=0.75}
           {arriveTangent=1, interpMode=RCIM_Cubic, leaveTangent=1, time=0.6, value=0.85}
           {arriveTangent=0.25, interpMode=RCIM_Cubic, leaveTangent=0.25, time=0.7, value=0.95}
           {arriveTangent=0.5, interpMode=RCIM_Cubic, leaveTangent=0.5, time=0.8, value=0.9}
           {arriveTangent=0.5, interpMode=RCIM_Cubic, leaveTangent=0.5, time=0.9, value=1.05}
           <at 1 1 RCIM_Cubic>
          ]}}}
       CurveIndex=&Multiply_Float(A=$Particles.NormalizedAge, B=2)
      )
     )
    )) {enabled=false}
   &ScaleMeshSize(Scale Factor=&MakeVector(
     X=1
     Y=1
     Z=&FloatFromCurve(
      FloatCurve=DataInterface() {propertyValues {Curve {keys [<at 0.3 0.8>, {arriveTangent=-1.1429, interpMode=RCIM_Cubic, leaveTangent=-1.1429, time=0.8, value=1.5}, <at 1 0>]}}}
      CurveIndex=$System.NormalizedLoopAge
      Scale Curve=0.7
     )
    )) {module=ScaleMeshSize002}
   &SolveForcesAndVelocity(Clamp Velocity=false, Limit Acceleration=false)
   /SetVariables(Particles.ColorInnerExhaust:LinearColor=&MakeLinearColorFromVectorAndFloat(
     "Vector (RGB)"=&MakeVectorFromLinearColorRGB(Color=&MultiplyLinearColor(
       Linear Color=&MakeCustomLinearColorFromBool(Bool=$System.IsDefaultVariant, True Color=<rgba 4 1.44 0.8 1>, False Color=<rgba 40 0 3.3333 1>)
       Scale Factor=&MakeVector4_FromLinearColor(LinearColor=$Particles.Color)
      ))
     "Float (Alpha)"=$System.IntensityOverTime
    )) {isSetParameters=true}
   /SetVariables(Particles.ColorOuterExhaust:LinearColor=&MakeLinearColorFromVectorAndFloat(
     "Vector (RGB)"=&MakeVectorFromLinearColorRGB(Color=&MultiplyLinearColor(
       Linear Color=&MakeCustomLinearColorFromBool(Bool=$System.IsDefaultVariant, True Color=<rgba 30 1.785 0.3 1>, False Color=<rgba 0.3 10.2 30 1>)
       Scale Factor=&MakeVector4_FromLinearColor(LinearColor=$Particles.Color)
      ))
     "Float (Alpha)"=$System.IntensityOverTime
    )) {isSetParameters=true}
   /SetVariables(Particles.ColorUpperExhaust:LinearColor=&MakeLinearColorFromVectorAndFloat(
     "Vector (RGB)"=&MakeVectorFromLinearColorRGB(Color=&MultiplyLinearColor(
       Linear Color=&MakeCustomLinearColorFromBool(Bool=$System.IsDefaultVariant, True Color=<rgba 1 0.875 0.7 1>, False Color=<rgba 0.7 0.8 1 1>)
       Scale Factor=&MakeVector4_FromLinearColor(LinearColor=$Particles.Color)
      ))
     "Float (Alpha)"=&FloatFromCurve(
      FloatCurve=DataInterface() {propertyValues {Curve {keys [{arriveTangent=-2.916, interpMode=RCIM_Cubic, leaveTangent=-2.916, tangentMode=RCTM_User, time=0, value=1}, <at 0.8 0 RCIM_Cubic>]}}}
      CurveIndex=$System.Timeline.NormalizedTime
      Scale Curve=0.2
     )
    )) {enabled=false, isSetParameters=true}
   &ScaleColor(
    Scale Alpha=&FloatFromCurve(
     FloatCurve=DataInterface() {propertyValues {Curve {keys [{arriveTangent=-2.916, interpMode=RCIM_Cubic, leaveTangent=-2.916, tangentMode=RCTM_User, time=0, value=1}, <at 0.8 0 RCIM_Cubic>]}}}
     CurveIndex=$System.Timeline.NormalizedTime
     Scale Curve=0.1
    )
    ScaleA=true
    ScaleRGB=false
   ) {enabled=false}
   &ScaleColor(Scale Alpha=$System.IntensityOverTime, ScaleA=true, ScaleRGB=true) {module=ScaleColor001}
  ]
  rendererClasses [&NiagaraMeshRendererProperties]
  renderers [
   {rendererClass=&NiagaraMeshRendererProperties, rendererIndex=0}
   {rendererClass=&NiagaraMeshRendererProperties, rendererIndex=1}
   {rendererClass=&NiagaraMeshRendererProperties, rendererIndex=2}
  ]
 }
 LaserBeam {
  bEnabled=false
  emitterUpdateScript [
   &EmitterState()
   &SpawnBurst_Instantaneous(Use Spawn Probability=false, UseLoopCountLimit=false)
  ]
  particleSpawnScript [
   &InitializeParticle(
    Color=<rgba 2.6135 1.672 4 0.05>
    Lifetime=5
    Position Offset Coordinate Space=::Local
    Sprite Size=<5 80>
    Sprite Size Mode=::NewEnumerator4
    UsePositionOffset=true
    Write Material Random=true
   )
   &AddVelocity(Velocity=<1 0 0>)
   &SkeletalMeshLocation(Bone Sampling Mode=::NewEnumerator4, MaskSpawnGroup=false, Skeletal Mesh=$User.SkeletalMesh) {enabled=false, module=SkeletalMeshLocation001}
   &SkeletalMeshLocation(Bone Sampling Mode=::NewEnumerator0, MaskSpawnGroup=false, OverrideRandomSeed=false, Skeletal Mesh=$User.SkeletalMesh) {enabled=false}
  ]
  particleUpdateScript [
   &ParticleState()
   &ScaleSpriteSize(
    Non-Uniform Curve Sprite Scale=DataInterface() {propertyValues {
      XCurve {keys [{tangentMode=RCTM_User, time=0.4, value=2}, {tangentMode=RCTM_User, time=0.8, value=0}]}
      YCurve {keys [{tangentMode=RCTM_User, time=0.2, value=0}, {tangentMode=RCTM_User, time=0.7, value=1}]}
     }}
    Scale Sprite Size Mode=::NewEnumerator3
   )
   &SolveForcesAndVelocity(Clamp Velocity=false, Limit Acceleration=false)
   &ScaleColor(Scale Alpha=$System.IntensityOverTime, ScaleA=true, ScaleRGB=true)
  ]
  rendererClasses [&NiagaraSpriteRendererProperties, &NiagaraMeshRendererProperties]
  renderers [
   {rendererClass=&NiagaraSpriteRendererProperties}
   {rendererClass=&NiagaraMeshRendererProperties, rendererIndex=1}
   {rendererClass=&NiagaraSpriteRendererProperties, rendererIndex=2}
  ]
 }
 Lights {
  emitterUpdateScript [
   &EmitterState(
    Inactive Response=::NewEnumerator1
    Life Cycle Mode=::NewEnumerator1
    Loop Behavior=::NewEnumerator1
    Loop Duration Mode=::NewEnumerator1
    UseLoopDelay=false
   )
   &SpawnBurst_Instantaneous(Spawn Count=2, Use Spawn Probability=false, UseLoopCountLimit=false)
  ]
  particleSpawnScript [
   &FN_SetSpawnGroup(Number of Spawn Groups=1)
   &InitializeParticle(Color=<rgba 31.6 4 40 1>, Position Offset=<40 0 0>, UsePositionOffset=true, Write Material Random=true)
   /SetVariables(Particles.LightRadius=&Multiply_Float(A=&MakeFloatFromVector(Vector=$Engine.Owner.Scale, Channel=::NewEnumerator0), B=166)) {isSetParameters=true}
  ]
  particleUpdateScript [
   &ParticleState(Kill Particles When Lifetime Has Elapsed=false)
   &ScaleColor(
    Scale Alpha=&Multiply_Float(
     A=&FloatFromCurve(
      FloatCurve=DataInterface() {propertyValues {Curve {keys [
          {arriveTangent=4.1785, interpMode=RCIM_Cubic, leaveTangent=4.1785, tangentMode=RCTM_User, time=0, value=0}
          <at 0.6898 1>
          {arriveTangent=1.1179, interpMode=RCIM_Cubic, leaveTangent=1.1179, tangentMode=RCTM_User, time=0.69, value=40}
          {arriveTangent=0.0224, interpMode=RCIM_Cubic, leaveTangent=0.0224, tangentMode=RCTM_User, time=0.8, value=1}
         ]}}}
      CurveIndex=$System.NormalizedLoopAge
      Scale Curve=&FNTimeOfDayScalar(Night Scale=0.1, Morning Scale=0.05, Day Scale=0.05, Evening Scale=0.1)
     )
     B=$System.IntensityOverTime
    )
    ScaleA=true
    ScaleRGB=false
   )
   &FN_Light_Attributes(Create Exponent Attribute=true, Create Radius Attribute=true, Exponent=4, Radius=100)
  ]
  rendererClasses [&NiagaraLightRendererProperties]
  renderers [{rendererClass=&NiagaraLightRendererProperties, rendererIndex=0}]
 }
 Smoke_Wave {
  emitterUpdateScript [
   &EmitterState(Life Cycle Mode=::NewEnumerator1, Loop Behavior=::NewEnumerator1, Loop Duration=3.2, UseLoopDelay=true) {module=EmitterState001}
   &self:ThresholdSpawnRate(MovementThreshold=50, SpawnRate=20) {enabled=false, module=ThresholdSpawnRate}
   /SetVariables(Emitter.BuildupCurve:NiagaraFloat=&FloatFromCurve(
     FloatCurve=DataInterface() {propertyValues {Curve {keys [
         {arriveTangent=0.0145, interpMode=RCIM_Cubic, leaveTangent=0.0145, tangentMode=RCTM_User, time=0, value=0}
         {arriveTangent=2.5625, interpMode=RCIM_Cubic, leaveTangent=2.5625, tangentMode=RCTM_User, time=0.8, value=1}
         {interpMode=RCIM_Cubic, leaveTangent=1.25, time=0.8373, value=0}
        ]}}}
     CurveIndex=$Emitter.NormalizedLoopAge
    )) {isSetParameters=true}
   /SetVariables(Emitter.BuildupCurveInfinite:NiagaraFloat=&FloatFromCurve(
     FloatCurve=DataInterface() {propertyValues {Curve {keys [
         {arriveTangent=0.0145, interpMode=RCIM_Cubic, leaveTangent=0.0145, tangentMode=RCTM_User, time=0, value=0}
         {arriveTangent=2.5625, interpMode=RCIM_Cubic, leaveTangent=2.5625, tangentMode=RCTM_User, time=1, value=1}
        ]}}}
     CurveIndex=$Emitter.NormalizedLoopAge
    )) {isSetParameters=true}
   &SpawnBurst_Instantaneous(Spawn Count=5, Use Spawn Probability=false, UseLoopCountLimit=false) {enabled=false}
   &SpawnRate(SpawnRate=20, Use Spawn Probability=false) {enabled=false}
   &SpawnRate(SpawnRate=&Multiply_Float(A=$Emitter.BuildupCurve, B=10), Use Spawn Probability=false) {module=SpawnRate001}
  ]
  particleSpawnScript [
   &InitializeParticle(
    Color=<rgba 0.4545 0.1085 1 0.05>
    Lifetime=3.5
    Position Offset=<35 0 0>
    Position Offset Coordinate Space=::Local
    Sprite Rotation Mode=::NewEnumerator2
    Sprite Size Mode=::NewEnumerator1
    Uniform Sprite Size Max=45
    Uniform Sprite Size Min=35
    UsePositionOffset=true
    Write Material Random=true
   ) {module=InitializeParticle002}
   /SetVariables(Particles.Rotation:NiagaraFloat=&RandomRangeFloat(Minimum=-80, Maximum=80)) {isSetParameters=true}
   /SetVariables(Particles.RandomFrame=&RandomRangeInt(Minimum=0, Maximum=63)) {isSetParameters=true}
   &SphereLocation(
    MaskSpawnGroup=false
    Non Uniform Scale=<2 0 0>
    Override Seed=false
    Sphere Radius=10
    Surface Only=false
    UseNonUniformScale=true
    UseOffset=false
    UseSphereOrientation=false
   ) {enabled=false}
   &ShapeLocation(
    MaskSpawnGroup=false
    Offset=&MakeVector(
     X=&FloatFromCurve(
      FloatCurve=DataInterface() {propertyValues {Curve {keys [{tangentMode=RCTM_User, time=0.1, value=0.2}, {tangentMode=RCTM_User, time=1, value=1}]}}}
      CurveIndex=$Emitter.BuildupCurve
      Scale Curve=200
     )
     Y=0
     Z=0
    )
    Offset Mode=::NewEnumerator0
    Override Seed=false
    Rotation Mode=::NewEnumerator2
    Sphere Radius=&FloatFromCurve(
     FloatCurve=DataInterface() {propertyValues {Curve {keys [{tangentMode=RCTM_User, time=0, value=0.1}, {tangentMode=RCTM_User, time=1, value=1}]}}}
     CurveIndex=$Emitter.BuildupCurveInfinite
     Scale Curve=150
    )
    UseOwnerScale=true
   ) {enabled=false}
   &CameraOffset()
   &AddVelocity(Velocity=&Lerp_Vector(
     A=<0 300 0>
     B=<0 -300 0>
     Alpha=&MakeCustomFloatFromBool(Bool=$User.bIsFrontEnd, True Float=1, False Float=0)
    )) {enabled=false}
   &DynamicMaterialParameters(Write Parameter Index 0=false)
  ]
  particleUpdateScript [
   &ParticleState()
   &SkeletalMeshLocation(
    Bone / Socket Index=&RandomRangeInt(Minimum=0, Maximum=5)
    MaskSpawnGroup=false
    Mesh Sampling Type=::NewEnumerator5
    Sampled Position Offset=&Lerp_Vector(
     A=Linked() {linkedVariable {name=Particles.Initial.Position, type=NiagaraPosition}}
     B=&AddVectorToPosition(
      Position=$Particles.Initial.Position
      Vector=&Multiply_Vector(
       A=&RandomRangeVector(Minimum=<0 50 0>, Maximum=<0 150 1>)
       B=&VectorFromFloat(Value=&Lerp_Float(A=1, B=-1, Alpha=&MakeCustomFloatFromBool(Bool=$User.bIsFrontEnd, True Float=1, False Float=0)))
      )
     )
     Alpha=&FloatFromCurve(FloatCurve=DataInterface() {propertyValues {Curve {keys [{tangentMode=RCTM_User, time=0, value=0}, {tangentMode=RCTM_User, time=1, value=1}]}}})
    )
    Sampled Position Offset Coordinate Space:ENiagaraSkelMeshTransforms=::NewEnumerator2
    Skeletal Mesh=$User.SkeletalMesh
    Socket Sampling Mode=::NewEnumerator5
   ) {enabled=false}
   &Drag(Drag=2, Use Linear Drag=true, Use Rotational Drag=false) {enabled=false}
   /SetVariables(Particles.Position=&LerpPosition(A=$Particles.Position, B=<25 0 0>, Alpha=$Particles.NormalizedAge)) {enabled=false, isSetParameters=true}
   &InheritSourceMovement(Applied Position Scale=<-1 -1 -1>, ApplyForce=false, ApplyPosition=true, ApplyPositionFalloff=false) {enabled=false}
   &SolveForcesAndVelocity(Clamp Velocity=false, Limit Acceleration=false)
   &ScaleColor(
    Scale Alpha=&FloatFromCurve(
     FloatCurve=DataInterface() {propertyValues {Curve {keys [
         <at 0 0 RCIM_Cubic>
         {interpMode=RCIM_Cubic, tangentMode=RCTM_User, time=0.2, value=1}
         {arriveTangent=-1.25, interpMode=RCIM_Cubic, leaveTangent=-1.25, time=0.4, value=0.9932}
         <at 1 0>
        ]}}}
     Scale Curve=&Multiply_Float(A=1, B=&FNTimeOfDayScalar(Night Scale=0.3, Morning Scale=0.1, Evening Scale=0.1))
    )
    ScaleA=true
    ScaleRGB=true
   ) {module=ScaleColor001}
   &ScaleColor(
    Scale Alpha=&FloatFromCurve(FloatCurve=DataInterface() {propertyValues {Curve {keys [
         {interpMode=RCIM_Cubic, tangentMode=RCTM_User, time=0, value=0}
         {interpMode=RCIM_Cubic, tangentMode=RCTM_User, time=0.2, value=1}
         {interpMode=RCIM_Cubic, tangentMode=RCTM_User, time=0.6, value=0}
        ]}}})
    ScaleA=true
    ScaleRGB=true
   )
   &ScaleColor(
    Scale Alpha=&VectorLength(Vector=&Clamp_Vector(
      Vector:Vector3f=&Subtract_Position(A=$Particles.Position, B=<43 0 0>)
      Min:Vector3f=<0 0 0>
      Max:Vector3f=<1 1 1>
     ))
    ScaleA=true
    ScaleRGB=true
   ) {module=ScaleColor002}
   &ScaleSpriteSize(
    Uniform Curve Scale=&Lerp_Float(A=0.7, B=1, Alpha=&MakeCustomFloatFromBool(Bool=$User.bIsFrontEnd, True Float=1, False Float=0))
    Uniform Curve Sprite Scale=DataInterface() {propertyValues {Curve {keys [
        {arriveTangent=0.46, interpMode=RCIM_Cubic, leaveTangent=0.46, tangentMode=RCTM_User, time=0, value=1}
        {arriveTangent=0.2144, interpMode=RCIM_Cubic, leaveTangent=0.2144, tangentMode=RCTM_User, time=1, value=1.2}
       ]}}}
   ) {module=ScaleSpriteSize001}
   &FN_SubUVAnimation(End Frame=&Add_Integer(A=$Particles.RandomFrame, B=63), Start Frame=$Particles.RandomFrame, SubUV Loop Count=0.3)
  ]
  rendererClasses [&NiagaraSpriteRendererProperties]
  renderers [{rendererClass=&NiagaraSpriteRendererProperties}]
 }
 SoftGlowSprite {
  emitterUpdateScript [
   &EmitterState(
    Inactive Response=::NewEnumerator1
    Life Cycle Mode=::NewEnumerator1
    Loop Behavior=::NewEnumerator1
    Loop Duration Mode=::NewEnumerator1
    UseLoopDelay=false
   )
   &SpawnBurst_Instantaneous(Use Spawn Probability=false, UseLoopCountLimit=false)
  ]
  particleSpawnScript [
   &FN_SetSpawnGroup(Number of Spawn Groups=1)
   &InitializeParticle(
    Color=<rgba 0.215 0.05 0.5 0.4>
    Position Offset=<35 0 0>
    Sprite Size Mode=::NewEnumerator3
    Uniform Sprite Size=&Multiply_Float(A=50, B=&MakeFloatFromVector(Vector=$Engine.Owner.Scale, Channel=::NewEnumerator0))
    UsePositionOffset=true
    Write Material Random=true
   )
   &CameraOffset()
  ]
  particleUpdateScript [
   &ParticleState(Kill Particles When Lifetime Has Elapsed=false)
   &ScaleColor(Scale Alpha=$System.IntensityOverTime, ScaleA=true, ScaleRGB=false)
   /SetVariables(Particles.New Linear Color:LinearColor=&MultiplyLinearColor(
     Linear Color=<rgba 1.177 0.354 2 4>
     Scale Factor=&MakeVector4(
      X=1
      Y=1
      Z=1
      W=&Multiply_Float(
       A=$System.IntensityOverTime
       B=&FloatFromCurve(
        FloatCurve=DataInterface() {propertyValues {Curve {keys [<at 3.5 1>,<at 4 0>]}}}
        Scale Curve=&FNTimeOfDayScalar(Night Scale=1, Morning Scale=1, Evening Scale=1)
       )
      )
     )
    )) {isSetParameters=true}
   &ScaleColor(
    Scale Alpha=&FloatFromCurve(FloatCurve=DataInterface() {propertyValues {Curve {keys [<at 2 1>,<at 3 0>]}}})
    ScaleA=true
    ScaleRGB=true
   ) {module=ScaleColor001}
  ]
  rendererClasses [&NiagaraSpriteRendererProperties]
  renderers [
   {rendererClass=&NiagaraSpriteRendererProperties}
   {rendererClass=&NiagaraSpriteRendererProperties, rendererIndex=1}
  ]
 }
}
summary {
 systemName=NS_Emote_HighMotion_Core
 emitters [
  {emitterName=AttractedOrbs, rendererClasses [&NiagaraSpriteRendererProperties]}
  {
   bEnabled=false
   emitterName=LaserBeam
   rendererClasses [&NiagaraSpriteRendererProperties, &NiagaraMeshRendererProperties]
  }
  {emitterName=SoftGlowSprite, rendererClasses [&NiagaraSpriteRendererProperties]}
  {emitterName=GlowCard, rendererClasses [&NiagaraSpriteRendererProperties]}
  {emitterName=Lights, rendererClasses [&NiagaraLightRendererProperties]}
  {emitterName=DysfunctionSpark, rendererClasses [&NiagaraSpriteRendererProperties]}
  {bEnabled=false, emitterName=GlowFlash, rendererClasses [&NiagaraSpriteRendererProperties]}
  {bEnabled=false, emitterName=AttractedSparks, rendererClasses [&NiagaraSpriteRendererProperties]}
  {emitterName=AnamorphicFlare, rendererClasses [&NiagaraSpriteRendererProperties]}
  {
   bEnabled=false
   emitterName=FN_Debug
   rendererClasses [&NiagaraSpriteRendererProperties, &NiagaraMeshRendererProperties]
  }
  {emitterName=GunGlow, rendererClasses [&NiagaraSpriteRendererProperties, &NiagaraLightRendererProperties]}
  {emitterName=HeatCones, rendererClasses [&NiagaraMeshRendererProperties]}
  {emitterName=Smoke_Wave, rendererClasses [&NiagaraSpriteRendererProperties]}
  {emitterName=Cylinder_Mesh, rendererClasses [&NiagaraMeshRendererProperties]}
  {emitterName=Debris, rendererClasses [&NiagaraSpriteRendererProperties]}
  {emitterName=DysfunctionSpark001, rendererClasses [&NiagaraSpriteRendererProperties]}
 ]
 userVariables [
  User.bIsFrontEnd {type=NiagaraBool}
  User.bIsFrontEndPreview {type=NiagaraBool}
  User.bIsInWater {type=NiagaraBool}
  User.Boost {type=NiagaraFloat, defaultValue=0, description=""}
  User.EffectScale {type=NiagaraFloat, defaultValue=1, description=""}
  User.EffectSpawnScale {type=NiagaraFloat, defaultValue=10, description=""}
  User.ForwardSpeed {type=NiagaraFloat, defaultValue=0, description=""}
  User.InWater {type=NiagaraBool}
  User.isBoosting {type=NiagaraBool}
  User.NewVariable {type=NiagaraFloat, defaultValue=0, description=""}
  User.SkeletalMesh {
   type=NiagaraDataInterfaceSkeletalMesh
   defaultValue {
    struct=NiagaraExt_VariableValue_DataInterface
    value {dataInterface=&self:NiagaraDataInterfaceSkeletalMesh_0}
   }
  }
  User.TargePosition {type=Vector3f, description="", defaultValue {struct=Vector3f, value=<0 0 0>}}
  User.TargetLockOn {type=NiagaraFloat, defaultValue=0, description=""}
  User.Variant {type=NiagaraFloat, defaultValue=1, description=""}
  User.WaterBodyIndex {type=NiagaraFloat, defaultValue=0, description=""}
  User.WaterDepth {type=NiagaraFloat, defaultValue=0, description=""}
  User.WaterPlaneLocation {type=Vector3f, description="", defaultValue {struct=Vector3f, value=<0 0 0>}}
  User.WaterPlaneNormal {type=Vector3f, description="", defaultValue {struct=Vector3f, value=<0 0 0>}}
  User.WaveAttenuationFactor {type=NiagaraFloat, defaultValue=0, description=""}
  User.WaveTime {type=NiagaraFloat, defaultValue=0, description=""}
 ]
}
