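# Niagara system for the intro of the HighMotion player emote, producing a magenta glow sprite with a matching light.
# LightAndGlow burst-spawns a particle whose alpha follows System.IntensityOverTime and whose color feeds the light.
# Dynamic material parameters pass the system's X axis and a quickly ramping curve value to the material.
# The disabled FN_Debug emitter writes an axis-to-camera dot product for debugging; the listed water and boost user parameters are not referenced by either emitter.
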NiagaraSystem /BRCosmetics/Animation/Game/MainPlayer/Emotes/HighMotion/FX/NS_Emote_HighMotion_Intro
dependencies {
 usedDataInterfaces [&NiagaraDataInterfaceCamera, &NiagaraDataInterfaceCurve]
 usedDynamicInputs [&Multiply_Float]
 usedModules [
  /Niagara/Modules/System/SystemLifeCycle
  &SystemState
  &self:SystemSpawnScript.SystemScriptSource.SystemScriptGraph.NiagaraNodeAssignment_0.SetVariables_#1
 ]
}
emitters {
 FN_Debug {
  bEnabled=false
  emitterUpdateScript [
   &EmitterState()
   &SpawnBurst_Instantaneous(Spawn Count=1, Use Spawn Probability=false)
  ]
  particleSpawnScript [
   /Niagara/Modules/Spawn/Initialization/InitializeParticle(
    Lifetime=5
    Module.Write Color=true
    Module.Write Lifetime=true
    Module.Write Mass=true
    Module.Write Material Random=true
    Module.Write Position=true
    Module.Write Scale=false
    Module.Write SpriteRotation=false
    Module.Write SpriteSize=true
    Sprite Size=<100 100>
   )
   /Niagara/Modules/Spawn/Initialization/V2/InitializeParticle(
    Mesh Scale Mode=::NewEnumerator3
    Mesh Uniform Scale=2
    Position Offset Coordinate Space=::Local
    UsePositionOffset=true
    Write Material Random=true
   ) {enabled=false, module=InitializeParticle001}
  ]
  particleUpdateScript [
   &ParticleState()
   /SetVariables(Particles.Debug:NiagaraFloat=&DotProduct(
     Vector A=/Game/Effects/Niagara/DynamicInputs/Engine/FN_GetSystemAxis(Axis:ENiagaraOrientationAxis=::XAxis, Flip Axis:NiagaraBool=false)
     Vector B=&ReturnCameraVector(Camera Vectors=::NewEnumerator1, Coordinate Space=::Local)
    )) {isSetParameters=true}
   &DynamicMaterialParameters(
    Index 0 Param 1=$Particles.Debug
    Module.Param0WriteEnabled=true
    Module.Param1WriteEnabled=true
    Module.Param2WriteEnabled=true
    Module.Param3WriteEnabled=true
   )
  ]
  rendererClasses [&NiagaraSpriteRendererProperties, &NiagaraMeshRendererProperties]
  renderers [
   {rendererClass=&NiagaraSpriteRendererProperties}
   {rendererClass=&NiagaraMeshRendererProperties, rendererIndex=1}
  ]
 }
 LightAndGlow {
  emitterUpdateScript [
   &EmitterState(
    Inactive Response=::NewEnumerator1
    Life Cycle Mode=::NewEnumerator1
    Loop Behavior=::NewEnumerator1
    Loop Duration Mode=::NewEnumerator1
    UseLoopDelay=false
   )
   &SpawnBurst_Instantaneous(Use Spawn Probability=false, UseLoopCountLimit=false)
  ]
  particleSpawnScript [&InitializeParticle(
    Color=<rgba 0.9175 0.01 1 1>
    Lifetime=5
    Sprite Size Mode=::NewEnumerator3
    Uniform Sprite Size=55
    UsePositionOffset=true
    Write Material Random=true
   )]
  particleUpdateScript [
   &ParticleState()
   &ScaleColor(Scale Alpha=$System.IntensityOverTime, Scale RGB=&VectorFromFloat(Value=2), ScaleA=true, ScaleRGB=true)
   &DynamicMaterialParameters(
    Index 0 Param 1=&MakeFloatFromVector(Vector=$Engine.Owner.SystemXAxis, Channel=::NewEnumerator0)
    Index 0 Param 2=&MakeFloatFromVector(Vector=$Engine.Owner.SystemXAxis, Channel=::NewEnumerator1)
    Index 0 Param 3=&MakeFloatFromVector(Vector=$Engine.Owner.SystemXAxis, Channel=::NewEnumerator2)
    Index 0 Param 4=&FloatFromCurve(
     FloatCurve=DataInterface() {propertyValues {Curve {keys [{interpMode=RCIM_Cubic, tangentMode=RCTM_User, time=0, value=0}, {interpMode=RCIM_Cubic, tangentMode=RCTM_User, time=0.2, value=1}]}}}
     Scale Curve=120
    )
    Module.Param0WriteEnabled=true
    Module.Param1WriteEnabled=true
    Module.Param2WriteEnabled=true
    Module.Param3WriteEnabled=true
   )
   &FN_Light_Attributes(Color=$Particles.Color, Create Color Attribute=true, Scale Color=0.5)
  ]
  rendererClasses [&NiagaraSpriteRendererProperties, &NiagaraLightRendererProperties]
  renderers [
   {rendererClass=&NiagaraSpriteRendererProperties}
   {rendererClass=&NiagaraLightRendererProperties, rendererIndex=1}
  ]
 }
}
summary {
 systemName=NS_Emote_HighMotion_Intro
 emitters [
  {
   bEnabled=false
   emitterName=FN_Debug
   rendererClasses [&NiagaraSpriteRendererProperties, &NiagaraMeshRendererProperties]
  }
  {emitterName=LightAndGlow, rendererClasses [&NiagaraSpriteRendererProperties, &NiagaraLightRendererProperties]}
 ]
 userVariables [
  User.bIsInWater {type=NiagaraBool}
  User.Boost {type=NiagaraFloat, defaultValue=0, description=""}
  User.EffectScale {type=NiagaraFloat, defaultValue=1, description=""}
  User.EffectSpawnScale {type=NiagaraFloat, defaultValue=10, description=""}
  User.ForwardSpeed {type=NiagaraFloat, defaultValue=0, description=""}
  User.InWater {type=NiagaraBool}
  User.isBoosting {type=NiagaraBool}
  User.NewVariable {type=NiagaraFloat, defaultValue=0, description=""}
  User.SkeletalMesh {
   type=NiagaraDataInterfaceSkeletalMesh
   defaultValue {
    struct=NiagaraExt_VariableValue_DataInterface
    value {dataInterface=&self:NiagaraDataInterfaceSkeletalMesh_0}
   }
  }
  User.TargePosition {type=Vector3f, description="", defaultValue {struct=Vector3f, value=<0 0 0>}}
  User.TargetLockOn {type=NiagaraFloat, defaultValue=0, description=""}
  User.WaterBodyIndex {type=NiagaraFloat, defaultValue=0, description=""}
  User.WaterDepth {type=NiagaraFloat, defaultValue=0, description=""}
  User.WaterPlaneLocation {type=Vector3f, description="", defaultValue {struct=Vector3f, value=<0 0 0>}}
  User.WaterPlaneNormal {type=Vector3f, description="", defaultValue {struct=Vector3f, value=<0 0 0>}}
  User.WaveAttenuationFactor {type=NiagaraFloat, defaultValue=0, description=""}
  User.WaveTime {type=NiagaraFloat, defaultValue=0, description=""}
 ]
}
