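# Niagara landing effect for the HighMotion emote, bursting impact spikes and two expanding smoke rings.
# Spikes and the smaller ring partition particles across skeletal mesh bones or sockets and push outward from them, while the big ring radiates from the simulation position.
# Both rings animate dynamic material parameters and spawn nothing when bIsInWater is set.
# Alpha scales with time of day; most other user variables, like pontoon and wave inputs, are unreferenced here.
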NiagaraSystem /BRCosmetics/Animation/Game/MainPlayer/Emotes/HighMotion/FX/NS_Emote_HighMotion_LandSmoke
dependencies {
 usedDataInterfaces [&NiagaraDataInterfaceCurve, &NiagaraDataInterfaceColorCurve, &NiagaraDataInterfaceVector2DCurve]
 usedDynamicInputs []
 usedModules [/Niagara/Modules/System/SystemLifeCycle, &SystemState]
}
emitters {
 Impact_Spikes {
  emitterUpdateScript [
   &EmitterState(Life Cycle Mode=::NewEnumerator1, Loop Behavior=::NewEnumerator1, Loop Delay=0.03, Loop Duration=1, UseLoopDelay=true)
   &SpawnBurst_Instantaneous(Spawn Count=13, Use Spawn Probability=false, UseLoopCountLimit=false)
  ]
  particleSpawnScript [
   &InitializeParticle(
    Color=&HueShiftLinearColor(Hue=0, Saturation=1, Value=&RandomRangeFloat(Minimum=0.4, Maximum=0.8), Linear Color=$Particles.Color)
    Lifetime Max=0.3
    Lifetime Min=0.2
    Lifetime Mode=::NewEnumerator1
    Sprite Size Max=<30 40>
    Sprite Size Min=<10 12.5>
    Sprite Size Mode=::NewEnumerator2
    UsePositionOffset=false
    Write Material Random=true
   )
   &PartitionParticles()
   &SkeletalMeshLocation(
    Bone / Socket Index=$Particles.PartitionParticles.Partition
    Bone Sampling Mode=::NewEnumerator4
    MaskSpawnGroup=false
    Skeletal Mesh=$User.SkeletalMesh
   )
   &TorusLocation(
    Align Torus With Axis=false
    Coordinate Space:ENiagaraCoordinateSpace=::World
    Handle Radius=0
    Large Radius=4
    MaskSpawnGroup=false
    Radius Position=0.5
    Torus Distribution Mode:ENiagaraTorusDistributionMode=::NewEnumerator1
    U Position:NiagaraFloat=&ReturnNormalizedExecIndex()
   )
   &AddVelocity(
    Module.Use Falloff Distance=false
    Origin Offset=<0 0 -2>
    Override Seed=false
    Velocity Mode=::NewEnumerator1
    Velocity Origin=$Particles.SkeletalMeshLocation.SampledPosition
    Velocity Speed=&RandomRangeFloat(Minimum=32, Maximum=64)
   )
   &AddVelocityFromPoint(
    Module.Use Falloff Distance=false
    Origin Offset=<0 0 -3>
    Origin Offset Coordinate Space=::World
    Velocity Strength=&RandomRangeFloat(Minimum=32, Maximum=64)
   ) {enabled=false}
   &SubUVAnimation(End Frame=3, SubUV Animation Mode:ENiagara_SubUVLookupMode=::NewEnumerator2)
  ]
  particleUpdateScript [
   &ParticleState()
   &Drag(Drag=3, Use Linear Drag=true, Use Rotational Drag=false)
   &SolveForcesAndVelocity(Clamp Velocity=false, Limit Acceleration=false)
   &ScaleSpriteSize(
    Non-Uniform Curve Scale=&Multiply_Vector2D_ByFloat(Float=2, Vector2D=<1 1>)
    Non-Uniform Curve Sprite Scale=DataInterface() {propertyValues {
      XCurve {keys [{tangentMode=RCTM_User, time=0, value=1}, {tangentMode=RCTM_User, time=1, value=0.8}]}
      YCurve {keys [{tangentMode=RCTM_User, time=0, value=0.5}, {tangentMode=RCTM_User, time=1, value=1.7}]}
     }}
    Scale Sprite Size Mode=::NewEnumerator3
   )
   &ScaleColor(
    Scale Alpha=&Multiply_Float(
     A=&FloatFromCurve(
      FloatCurve=DataInterface() {propertyValues {Curve {keys [<at 0 0>,<at 0.1 1>,<at 0.2 0.8>,<at 1 0>]}}}
      Scale Curve=&FNTimeOfDayScalar(Night Scale=1, Morning Scale=0.3, Evening Scale=0.5)
     )
     B=1
    )
    Scale RGB=<1 0.8 0.8>
    ScaleA=true
    ScaleRGB=true
   )
   /SetVariables(Particles.Color2=&MultiplyLinearColor(Linear Color=$Particles.Color, Scale Factor=<1 1 1 0.2>)) {isSetParameters=true}
   &SpriteFacingAndAlignment(Facing Coordinate Space=::World, Sprite Facing=<0 0 1>, WriteAlignment=false, WriteFacing=true)
  ]
  rendererClasses [&NiagaraSpriteRendererProperties]
  renderers [
   {rendererClass=&NiagaraSpriteRendererProperties}
   {rendererClass=&NiagaraSpriteRendererProperties, rendererIndex=1}
  ]
 }
 SmokeSmallerRing {
  emitterUpdateScript [
   &EmitterState()
   &SpawnBurst_Instantaneous(
    Spawn Count=&Multiply_Int(A=&MakeCustomIntFromBool(Bool=$User.bIsInWater, True INT=0, False INT=1), B=15)
    Spawn Time=0.1
    Use Spawn Probability=false
   ) {module=SpawnBurst_Instantaneous001}
  ]
  particleSpawnScript [
   /Niagara/Modules/Spawn/Initialization/InitializeParticle(
    Color=<rgba 0.2656 0.2656 0.2656 0.1>
    Lifetime=&UniformRangedFloat(Minimum=0.4, Maximum=0.8)
    Module.Write Color=true
    Module.Write Lifetime=true
    Module.Write Mass=true
    Module.Write Material Random=true
    Module.Write Position=false
    Module.Write Scale=false
    Module.Write SpriteRotation=true
    Module.Write SpriteSize=true
    Position=&SimulationPosition()
    Sprite Rotation=&UniformRangedFloat(Minimum=0, Maximum=360)
    Sprite Size=&Vector2DFromFloat(Value=&UniformRangedFloat(Minimum=50, Maximum=70))
   )
   &PartitionParticles()
   &SkeletalMeshLocation(
    Bone / Socket Index=$Particles.PartitionParticles.Partition
    Bone Sampling Mode=::NewEnumerator4
    MaskSpawnGroup=false
    Skeletal Mesh=$User.SkeletalMesh
   )
   &TorusLocation(
    Align Torus With Axis=true
    Large Radius=13
    MaskSpawnGroup=false
    Radius Position=0.5
    Torus Distribution Mode:ENiagaraTorusDistributionMode=::NewEnumerator1
    Torus Mode=::NewEnumerator2
    U Position:NiagaraFloat=&ReturnNormalizedExecIndex()
   )
   &AddVelocity(
    Module.Use Falloff Distance=false
    Override Seed=false
    Velocity Mode=::NewEnumerator1
    Velocity Origin=$Particles.SkeletalMeshLocation.SampledPosition
    Velocity Speed=&UniformRangedFloat(Minimum=200, Maximum=50)
   )
   &AddVelocityFromPoint(
    Module.Use Falloff Distance=false
    Velocity Origin:Vector3f=&SimulationPosition()
    Velocity Strength=&UniformRangedFloat(Minimum=400, Maximum=550)
   ) {enabled=false}
   &ApplyInitialForces()
  ]
  particleUpdateScript [
   &ParticleState()
   &AccelerationForce(Acceleration=&MakeVector(X=0, Y=0, Z=&UniformRangedFloat(Minimum=10, Maximum=50)))
   &Drag(Drag=4, Use Linear Drag=true, Use Rotational Drag=false)
   &SolveForcesAndVelocity(Clamp Velocity=false, Limit Acceleration=false)
   &ScaleSpriteSize(Scale Factor=&Vector2DFromFloat(Value=&FloatFromCurve(
      FloatCurve=DataInterface() {propertyValues {Curve {keys [
          {arriveTangent=7.1593, interpMode=RCIM_Cubic, leaveTangent=7.1593, tangentMode=RCTM_User, time=0, value=0}
          {arriveTangent=1.1699, interpMode=RCIM_Cubic, leaveTangent=1.1699, tangentMode=RCTM_User, time=0.31, value=0.97}
          <at 1 1.25 RCIM_Cubic>
         ]}}}
      Scale Curve=0.5
     )))
   /SetVariables(Particles.SubImageIndex=&FloatFromCurve(FloatCurve=DataInterface() {propertyValues {Curve {keys [
         {arriveTangent=29.8623, interpMode=RCIM_Cubic, leaveTangent=29.8623, tangentMode=RCTM_User, time=-0.0046, value=5.0556}
         {arriveTangent=8.1241, interpMode=RCIM_Cubic, leaveTangent=8.1241, tangentMode=RCTM_User, time=1.0011, value=22.1689}
        ]}}})) {enabled=false, isSetParameters=true, module=SetVariables_#1}
   &DynamicMaterialParameters(
    Index 0 Param 1=&FloatFromCurve(
     FloatCurve=DataInterface() {propertyValues {Curve {keys [
         {arriveTangent=0.9088, interpMode=RCIM_Cubic, leaveTangent=0.9088, tangentMode=RCTM_User, time=0, value=1}
         {arriveTangent=0.5473, interpMode=RCIM_Cubic, leaveTangent=0.5473, tangentMode=RCTM_User, time=1, value=2}
        ]}}}
     Scale Curve=0.2
    )
    Index 0 Param 3=&FloatFromCurve(FloatCurve=DataInterface() {propertyValues {Curve {keys [
         {arriveTangent=-9.0447, interpMode=RCIM_Cubic, leaveTangent=-2.0044, tangentMode=RCTM_User, time=0, value=0.5}
         {interpMode=RCIM_Cubic, leaveTangent=4.6779, time=0.2851, value=0}
        ]}}})
    Index 1 Param 1=&FloatFromCurve(
     FloatCurve=DataInterface() {propertyValues {Curve {keys [<at 0 0>,<at 0.25 1>]}}}
     Scale Curve=&RandomRangeFloat(Minimum=2, Maximum=1)
    )
    Index 1 Param 2=&FloatFromCurve(
     FloatCurve=DataInterface() {propertyValues {Curve {keys [
         {interpMode=RCIM_Cubic, leaveTangent=3.7422, tangentMode=RCTM_Break, time=0, value=0.2}
         {interpMode=RCIM_Cubic, leaveTangent=0.9066, time=1, value=2}
        ]}}}
     Scale Curve=&RandomRangeFloat(Minimum=0.5, Maximum=1)
    )
    Index 1 Param 3=&FloatFromCurve(FloatCurve=DataInterface() {propertyValues {Curve {keys [
         {interpMode=RCIM_Cubic, leaveTangent=2.8445, tangentMode=RCTM_Break, time=0, value=0.2}
         {interpMode=RCIM_Cubic, leaveTangent=63.2978, time=0.25, value=1}
        ]}}})
    Index 1 Param 4=&FloatFromCurve(
     FloatCurve=DataInterface() {propertyValues {Curve {keys [<at 0 1>,<at 1 0.8>]}}}
     Scale Curve=&RandomRangeFloat(Minimum=0.5, Maximum=1)
    )
    Index 2 Param 1=&FloatFromCurve(
     FloatCurve=DataInterface() {propertyValues {Curve {keys [{interpMode=RCIM_Cubic, leaveTangent=3.2309, tangentMode=RCTM_Break, time=0.1, value=0}, <at 1 1>]}}}
     Scale Curve=0
    )
    Index 2 Param 2=&FloatFromCurve(
     FloatCurve=DataInterface() {propertyValues {Curve {keys [<at 0 0.2 RCIM_Cubic>,<at 1 0.8 RCIM_Cubic>]}}}
     Scale Curve=&RandomRangeFloat(Minimum=0.2, Maximum=1.2)
    )
    Index 2 Param 3=&FloatFromCurve(
     FloatCurve=DataInterface() {propertyValues {Curve {keys [
         {interpMode=RCIM_Cubic, tangentMode=RCTM_Break, time=0, value=1}
         {interpMode=RCIM_Cubic, leaveTangent=11.1233, tangentMode=RCTM_Break, time=1, value=2}
        ]}}}
     Scale Curve=0.1
    )
    Index 2 Param 4=&FloatFromCurve(
     FloatCurve=DataInterface() {propertyValues {Curve {keys [<at 0 1 RCIM_Cubic>, {interpMode=RCIM_Cubic, leaveTangent=-1.2956, tangentMode=RCTM_Break, time=1, value=0.75}]}}}
     Scale Curve=0.8
    )
    Module.Param0WriteEnabled=true
    Module.Param0WriteEnabled1=true
    Module.Param0WriteEnabled2=true
    Module.Param1WriteEnabled=true
    Module.Param1WriteEnabled1=true
    Module.Param1WriteEnabled2=true
    Module.Param2WriteEnabled=true
    Module.Param2WriteEnabled1=true
    Module.Param2WriteEnabled2=true
    Module.Param3WriteEnabled=true
    Module.Param3WriteEnabled1=true
    Module.Param3WriteEnabled2=true
    Write Parameter Index 1=true
    Write Parameter Index 2=true
   )
   &ScaleColor(
    Scale Alpha=&Multiply_Float(
     A=&FloatFromCurve(FloatCurve=DataInterface() {propertyValues {Curve {keys [<at 0 0>,<at 0.2049 0.9765>,<at 0.7791 0.9107>,<at 1 0>]}}}, Scale Curve=1.75)
     B=&Multiply_Float(A=&FNTimeOfDayScalar(Night Scale=1, Morning Scale=0.3, Evening Scale=0.3), B=1)
    )
    Scale RGB=&VectorFromFloat(Value=&FloatFromCurve(FloatCurve=DataInterface() {propertyValues {Curve {keys [<at 0 1>,<at 0.9983 0.5>]}}}))
    ScaleA=true
    ScaleRGB=true
   )
  ]
  rendererClasses [&NiagaraSpriteRendererProperties]
  renderers [{rendererClass=&NiagaraSpriteRendererProperties}]
 }
 Smoke_BigRing {
  emitterUpdateScript [
   &EmitterState()
   &SpawnBurst_Instantaneous(
    Spawn Count=&Multiply_Int(A=&MakeCustomIntFromBool(Bool=$User.bIsInWater, True INT=0, False INT=1), B=15)
    Spawn Time=0.16
    Use Spawn Probability=false
   ) {module=SpawnBurst_Instantaneous001}
  ]
  particleSpawnScript [
   &InitializeParticle(
    Color=<rgba 0.2656 0.2656 0.2656 0.1>
    Lifetime=&UniformRangedFloat(Minimum=0.4, Maximum=1.6)
    Module.Write Color=true
    Module.Write Lifetime=true
    Module.Write Mass=true
    Module.Write Material Random=true
    Module.Write Position=false
    Module.Write Scale=false
    Module.Write SpriteRotation=true
    Module.Write SpriteSize=true
    Position=&SimulationPosition()
    Sprite Rotation=&UniformRangedFloat(Minimum=0, Maximum=360)
    Sprite Size=&Vector2DFromFloat(Value=&UniformRangedFloat(Minimum=50, Maximum=70))
   )
   &TorusLocation(
    Align Torus With Axis=true
    Large Radius=3
    MaskSpawnGroup=false
    Torus Distribution Mode:ENiagaraTorusDistributionMode=::NewEnumerator1
    Torus Mode=::NewEnumerator2
    U Position:NiagaraFloat=&ReturnNormalizedExecIndex()
   )
   &AddVelocityFromPoint(
    Module.Use Falloff Distance=false
    Origin Offset=<0 0 0.7>
    Velocity Origin:Vector3f=&SimulationPosition()
    Velocity Strength=&Multiply_Float(A=&UniformRangedFloat(Minimum=800, Maximum=1200), B=0.6)
   )
   &ApplyInitialForces()
  ]
  particleUpdateScript [
   &ParticleState()
   &AccelerationForce(Acceleration=&MakeVector(X=0, Y=0, Z=&UniformRangedFloat(Minimum=10, Maximum=50)))
   &Drag(Drag=5, Use Linear Drag=true, Use Rotational Drag=false)
   &SolveForcesAndVelocity(Clamp Velocity=false, Limit Acceleration=false)
   &ScaleSpriteSize(Scale Factor=&Vector2DFromFloat(Value=&FloatFromCurve(FloatCurve=DataInterface() {propertyValues {Curve {keys [
          {arriveTangent=7.1593, interpMode=RCIM_Cubic, leaveTangent=7.1593, tangentMode=RCTM_User, time=0, value=0}
          {arriveTangent=1.1699, interpMode=RCIM_Cubic, leaveTangent=1.1699, tangentMode=RCTM_User, time=0.31, value=0.97}
          <at 1 1.25 RCIM_Cubic>
         ]}}})))
   /SetVariables(Particles.SubImageIndex=&FloatFromCurve(FloatCurve=DataInterface() {propertyValues {Curve {keys [
         {arriveTangent=29.8623, interpMode=RCIM_Cubic, leaveTangent=29.8623, tangentMode=RCTM_User, time=-0.0046, value=5.0556}
         {arriveTangent=8.1241, interpMode=RCIM_Cubic, leaveTangent=8.1241, tangentMode=RCTM_User, time=1.0011, value=22.1689}
        ]}}})) {enabled=false, isSetParameters=true, module=SetVariables_#1}
   &DynamicMaterialParameters(
    Index 0 Param 1=&FloatFromCurve(
     FloatCurve=DataInterface() {propertyValues {Curve {keys [
         {arriveTangent=0.9088, interpMode=RCIM_Cubic, leaveTangent=0.9088, tangentMode=RCTM_User, time=0, value=1}
         {arriveTangent=0.5473, interpMode=RCIM_Cubic, leaveTangent=0.5473, tangentMode=RCTM_User, time=1, value=2}
        ]}}}
     Scale Curve=0.5
    )
    Index 0 Param 3=&FloatFromCurve(
     FloatCurve=DataInterface() {propertyValues {Curve {keys [
         {arriveTangent=-9.0447, interpMode=RCIM_Cubic, leaveTangent=-2.0044, tangentMode=RCTM_User, time=0, value=0.5}
         {interpMode=RCIM_Cubic, leaveTangent=4.6779, time=0.2851, value=0}
        ]}}}
     Scale Curve=0
    )
    Index 1 Param 1=&FloatFromCurve(
     FloatCurve=DataInterface() {propertyValues {Curve {keys [<at 0 0>,<at 0.25 1>]}}}
     Scale Curve=&RandomRangeFloat(Minimum=2, Maximum=1)
    )
    Index 1 Param 2=&FloatFromCurve(
     FloatCurve=DataInterface() {propertyValues {Curve {keys [
         {interpMode=RCIM_Cubic, leaveTangent=3.7422, tangentMode=RCTM_Break, time=0, value=0.2}
         {interpMode=RCIM_Cubic, leaveTangent=0.9066, time=1, value=2}
        ]}}}
     Scale Curve=&RandomRangeFloat(Minimum=0.5, Maximum=1)
    )
    Index 1 Param 3=&FloatFromCurve(FloatCurve=DataInterface() {propertyValues {Curve {keys [
         {interpMode=RCIM_Cubic, leaveTangent=2.8445, tangentMode=RCTM_Break, time=0, value=0.2}
         {interpMode=RCIM_Cubic, leaveTangent=63.2978, time=0.25, value=1}
        ]}}})
    Index 1 Param 4=&FloatFromCurve(
     FloatCurve=DataInterface() {propertyValues {Curve {keys [<at 0 1>,<at 1 0.8>]}}}
     Scale Curve=&RandomRangeFloat(Minimum=0.5, Maximum=1)
    )
    Index 2 Param 1=&FloatFromCurve(
     FloatCurve=DataInterface() {propertyValues {Curve {keys [{interpMode=RCIM_Cubic, leaveTangent=3.2309, tangentMode=RCTM_Break, time=0.1, value=0}, <at 1 1>]}}}
     Scale Curve=0
    )
    Index 2 Param 2=&FloatFromCurve(
     FloatCurve=DataInterface() {propertyValues {Curve {keys [<at 0 0.2 RCIM_Cubic>,<at 1 0.8 RCIM_Cubic>]}}}
     Scale Curve=&RandomRangeFloat(Minimum=0.2, Maximum=1.2)
    )
    Index 2 Param 3=&FloatFromCurve(
     FloatCurve=DataInterface() {propertyValues {Curve {keys [
         {interpMode=RCIM_Cubic, tangentMode=RCTM_Break, time=0, value=1}
         {interpMode=RCIM_Cubic, leaveTangent=11.1233, tangentMode=RCTM_Break, time=1, value=2}
        ]}}}
     Scale Curve=0.1
    )
    Index 2 Param 4=&FloatFromCurve(
     FloatCurve=DataInterface() {propertyValues {Curve {keys [<at 0 1 RCIM_Cubic>, {interpMode=RCIM_Cubic, leaveTangent=-1.2956, tangentMode=RCTM_Break, time=1, value=0.75}]}}}
     Scale Curve=0.8
    )
    Module.Param0WriteEnabled=true
    Module.Param0WriteEnabled1=true
    Module.Param0WriteEnabled2=true
    Module.Param1WriteEnabled=true
    Module.Param1WriteEnabled1=true
    Module.Param1WriteEnabled2=true
    Module.Param2WriteEnabled=true
    Module.Param2WriteEnabled1=true
    Module.Param2WriteEnabled2=true
    Module.Param3WriteEnabled=true
    Module.Param3WriteEnabled1=true
    Module.Param3WriteEnabled2=true
    Write Parameter Index 1=true
    Write Parameter Index 2=true
   )
   &ScaleColor(
    Scale Alpha=&Multiply_Float(
     A=&FloatFromCurve(FloatCurve=DataInterface() {propertyValues {Curve {keys [<at 0 0>,<at 0.2049 0.9765>,<at 0.7791 0.9107>,<at 1 0>]}}}, Scale Curve=0.65)
     B=&Multiply_Float(A=&FNTimeOfDayScalar(Night Scale=1, Morning Scale=0.3, Evening Scale=0.3), B=1)
    )
    Scale RGB=&VectorFromFloat(Value=&FloatFromCurve(FloatCurve=DataInterface() {propertyValues {Curve {keys [<at 0 1>,<at 0.9983 0.5>]}}}))
    ScaleA=true
    ScaleRGB=true
   )
  ]
  rendererClasses [&NiagaraSpriteRendererProperties]
  renderers [{rendererClass=&NiagaraSpriteRendererProperties}]
 }
}
summary {
 systemName=NS_Emote_HighMotion_LandSmoke
 emitters [
  {emitterName=Smoke_BigRing, rendererClasses [&NiagaraSpriteRendererProperties]}
  {emitterName=Impact_Spikes, rendererClasses [&NiagaraSpriteRendererProperties]}
  {emitterName=SmokeSmallerRing, rendererClasses [&NiagaraSpriteRendererProperties]}
 ]
 userVariables [
  User.bIsInWater {type=NiagaraBool}
  User.Boost {type=NiagaraFloat, defaultValue=0, description=""}
  User.EffectScale {type=NiagaraFloat, defaultValue=1, description=""}
  User.EffectSpawnScale {type=NiagaraFloat, defaultValue=10, description=""}
  User.ForwardSpeed {type=NiagaraFloat, defaultValue=0, description=""}
  User.FrontPontoonBig.Depth {type=NiagaraFloat, defaultValue=0, description=""}
  User.FrontPontoonBig.Impulse {type=Vector3f, description="", defaultValue {struct=Vector3f, value=<0 0 0>}}
  User.InWater {type=NiagaraBool}
  User.isBoosting {type=NiagaraBool}
  User.MidPontoonBig.Depth {type=NiagaraFloat, defaultValue=0, description=""}
  User.MidPontoonBig.Impulse {type=Vector3f, description="", defaultValue {struct=Vector3f, value=<0 0 0>}}
  User.NewVariable {type=NiagaraFloat, defaultValue=0, description=""}
  User.RearPontoonBig.Depth {type=NiagaraFloat, defaultValue=0, description=""}
  User.RearPontoonBig.Impulse {type=Vector3f, description="", defaultValue {struct=Vector3f, value=<0 0 0>}}
  User.SkeletalMesh {
   type=NiagaraDataInterfaceSkeletalMesh
   defaultValue {
    struct=NiagaraExt_VariableValue_DataInterface
    value {dataInterface=&self:NiagaraDataInterfaceSkeletalMesh_0}
   }
  }
  User.WaterBodyIndex {type=NiagaraFloat, defaultValue=0, description=""}
  User.WaterDepth {type=NiagaraFloat, defaultValue=0, description=""}
  User.WaterPlaneLocation {type=Vector3f, description="", defaultValue {struct=Vector3f, value=<0 0 0>}}
  User.WaterPlaneNormal {type=Vector3f, description="", defaultValue {struct=Vector3f, value=<0 0 0>}}
  User.WaveAttenuationFactor {type=NiagaraFloat, defaultValue=0, description=""}
  User.WaveTime {type=NiagaraFloat, defaultValue=0, description=""}
 ]
}
